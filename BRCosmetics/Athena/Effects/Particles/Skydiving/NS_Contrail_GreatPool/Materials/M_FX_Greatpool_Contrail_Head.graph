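# Translucent unlit Niagara material for the skull head of the GreatPool skydiving contrail, drawn as animated electricity.
# Noise-distorted UVs sample skull textures, which combine with timed gradient reveals, radial bolts and a Fresnel outline into a lightning distance field.
# Thresholds on that field form bolt glow and a bright core tinted by particle color.
# Dynamic parameters drive intensity, bolt thickness, two-sidedness and jaw rotation about a pivot, with vertex green marking the jaw.
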Material /BRCosmetics/Athena/Effects/Particles/Skydiving/NS_Contrail_GreatPool/Materials/M_FX_Greatpool_Contrail_Head
graph {
 properties {
  bAutomaticallySetUsageInEditor=false
  bUsedWithNiagaraMeshParticles=true
  bUsedWithNiagaraSprites=true
  blendMode=BLEND_Translucent
  pixelDepthOffsetMode=PDOM_Legacy
  shadingModel=MSM_Unlit
  twoSided=true
  usedShadingModels=Unlit
 }
 nodes [
  CameraVectorWS()
  n1 = ScalarParameter() {parameterName=dbg_2sidedsign}
  LinearInterpolate(Alpha=@n1)
  n2 = VertexInterpolator(VS=DynamicParameter() {defaultValue=<rgba 3 1 0.5 1>, paramNames [IntensityRamp, JawAngle, BoltThickness, Use2Sided]}.Use2Sided)
  LinearInterpolate(B=LinearInterpolate(A=Constant3Vector() {constant=<rgba 1 0 0 0>}, B=Constant3Vector() {constant=<rgba 0 1 0 0>}, Alpha=@n2.PS), Alpha=@n1)
  n3 = ScalarParameter() {defaultValue=0.78, parameterName=ThinBoltCore}
  n4 = NamedRerouteUsage()
  n5 = SmoothStep(Min=@n3, Max=Add(A=@n3) {constB=0.01}, Value=@n4) {constMax=0.55, constMin=0.5}
  n6 = Constant3Vector() {constant=<rgba 33 33 33 0>}
  MaterialFunctionCall(Emissive=Add(A=ParticleColor().RGB, B=Multiply(A=@n5, B=@n6))) {materialFunction=&MF_FX_EmissiveAdjustments}
  n7 = DynamicParameter() {defaultValue=<rgba 3 1 0.5 1>, paramNames [IntensityRamp, JawAngle, BoltThickness, Use2Sided]}
  n8 = Add(A=@n7.BoltThickness) {constB=0.05}
  Multiply(
   A=Add(A=Power(Base=@n4) {constExponent=4.5}, B=SmoothStep(Min=@n7.BoltThickness, Max=@n8, Value=@n4) {constMax=0.55, constMin=0.5})
   B=NamedRerouteUsage()
  )
  n9 = Transform(None=Normalize(VectorInput=VectorParameter() {group=JawDynamics, parameterName=JawAxis, defaultValue=<rgba 0 0 0 1>}.RGB)) {transformSourceType=TRANSFORMSOURCE_Instance}
  n10 = VectorParameter() {group=JawDynamics, parameterName=JawPivot, defaultValue=<rgba 0 0 0 1>}
  Multiply(
   A=NamedRerouteUsage()
   B=RotateAboutAxis(
    NormalizedRotationAxis=Normalize(VectorInput=@n9)
    RotationAngle=DynamicParameter() {defaultValue=<rgba 3 1 0.5 1>, paramNames [IntensityRamp, JawAngle, BoltThickness, Use2Sided]}.JawAngle
    PivotPoint=TransformPosition(None=@n10.RGB) {transformSourceType=TRANSFORMPOSSOURCE_Instance, transformType=TRANSFORMPOSSOURCE_World}
    Position=WorldPosition().XYZ
   )
  )
  n11 = TwoSidedSign()
  NamedRerouteDeclaration(None=LinearInterpolate(A=Saturate(None=@n11), Alpha=@n2.PS)) {name=AngleDependant2SidedAlpha, nodeColor=<rgba 0.3125 1 0 1>}
  n12 = NamedRerouteUsage()
  n13 = TextureSample(UVs=@n12) {
   samplerType=SAMPLERTYPE_Masks
   texture=/BRCosmetics/Athena/Effects/Particles/Skydiving/NS_Contrail_GreatPool/Meshes/T_FX_GreatPool_SkullLayout
  }
  NamedRerouteDeclaration(None=@n13.B) {name=OpacityMask, nodeColor=<rgba 0.0547 1 0 1>}
  n14 = ScalarParameter() {defaultValue=0.05, parameterName=FresnelRadius}
  n15 = NamedRerouteUsage()
  n16 = Multiply(
   A=SphereMask(
    A=MaterialFunctionCall() {materialFunction=&MF_Fresnel_FX}.Fresnel Inverted Unclamped
    B=ScalarParameter() {defaultValue=0.7, parameterName=FresnelPos}
    Radius=Multiply(A=@n14, B=@n15)
   ) {hardnessPercent=0}
   B=OneMinus(None=VertexColor().R)
  )
  NamedRerouteDeclaration(None=Multiply(A=Saturate(None=@n16)) {constB=0.85}) {name=FresnelOutline, nodeColor=<rgba 1 0.0938 0 1>}
  n17 = DynamicParameter() {defaultValue=<rgba 3 1 0.5 1>, paramNames [IntensityRamp, JawAngle, BoltThickness, Use2Sided]}
  n18 = MaterialFunctionCall(Vector or UVs=Subtract(A=TextureCoordinate() {coordinateIndex=2}) {constB=0.5}) {materialFunction=&VectorToRadialValue}
  n19 = Multiply(
   A=@n18.Radial Coordinates
   B=ComponentMask(None=VectorParameter() {group=IntenseLayer, parameterName=IntenseElec_Tile, defaultValue=<rgba 1 1 0 1>}.RGB) {g=true, r=true}
  )
  n20 = ComponentMask(None=VectorParameter() {group=IntenseLayer, parameterName=IntenseElec_Speed, defaultValue=<rgba 0 1 0 1>}.RGB) {g=true, r=true}
  n21 = ScalarParameter() {group=IntenseLayer, parameterName=IntenseLayerDistortion}
  n22 = MaterialFunctionCall(XY=TextureCoordinate() {coordinateIndex=1}) {materialFunction=&Swizzle}
  n23 = ComponentMask(None=VectorParameter() {group=DetailNoise, parameterName=DetailNoise_Tile, defaultValue=<rgba 1 1 0 1>}.RGB) {g=true, r=true}
  n24 = VectorParameter() {group=DetailNoise, parameterName=DetailNoise_Speed, defaultValue=<rgba 0 1 0 1>}
  n25 = TextureSampleParameter2D(UVs=Panner(Coordinate=Multiply(A=@n22.YX, B=@n23), Speed=ComponentMask(None=@n24.RGB) {g=true, r=true}) {bFractionalPart=true}) {
   group=DetailNoise
   parameterName=T_NoiseTex
   samplerType=SAMPLERTYPE_LinearColor
   texture=/Game/Characters/Player/Male/Medium/Bodies/M_MED_SRDrift/Textures/T_ElectricNoise
  }
  n26 = ComponentMask(None=VectorParameter() {group=DetailNoise, parameterName=DetailNoise2_Tile, defaultValue=<rgba 1 1 0 1>}.RGB) {g=true, r=true}
  n27 = VectorParameter() {group=DetailNoise, parameterName=DetailNoise2_Speed, defaultValue=<rgba 0 1 0 1>}
  n28 = TextureSampleParameter2D(UVs=Panner(Coordinate=Multiply(A=@n22.YX, B=@n26), Speed=ComponentMask(None=@n27.RGB) {g=true, r=true}) {bFractionalPart=true}) {
   group=DetailNoise
   parameterName=T_NoiseTex
   samplerType=SAMPLERTYPE_LinearColor
   texture=&T_ElectricNoise
  }
  n29 = Divide(A=Add(A=@n25.R, B=@n28.R))
  n30 = Subtract(A=Multiply(A=NamedRerouteDeclaration(None=@n29) {name=NoiseLayer, nodeColor=<rgba 1 0 0.4453 1>}) {constB=2})
  n31 = TextureSample(UVs=Add(A=Panner(Coordinate=@n19, Speed=@n20) {bFractionalPart=true}, B=Multiply(A=@n21, B=@n30))) {
   samplerType=SAMPLERTYPE_LinearGrayscale
   texture=/BRCosmetics/Athena/Effects/Particles/Skydiving/NS_Contrail_GreatPool/Meshes/T_FX_GreatPool_Bolts
  }
  NamedRerouteDeclaration(None=Multiply(A=NamedRerouteUsage(), B=Multiply(A=@n17.IntensityRamp, B=@n31.R))) {name=IntenseElectricity, nodeColor=<rgba 0 1 0.1562 1>}
  n32 = ComponentMask(None=VectorParameter() {group=DetailNoise, parameterName=DetailNoise_Distortion, defaultValue=<rgba 0 0 0 1>}.RGB) {g=true, r=true}
  NamedRerouteDeclaration(None=Add(A=TextureCoordinate(), B=Multiply(A=@n30, B=@n32))) {name=DistortedUV, nodeColor=<rgba 0.5234 1 0 1>}
  NamedRerouteDeclaration(None=@n13.G) {name=IntenseMaskout, nodeColor=<rgba 0.5469 1 0 1>}
  NamedRerouteDeclaration(None=VertexColor().G) {name=IsJaw, nodeColor=<rgba 1 0.6797 0 1>}
  n33 = TextureSample(UVs=@n12) {
   samplerType=SAMPLERTYPE_LinearColor
   texture=/BRCosmetics/Athena/Effects/Particles/Skydiving/NS_Contrail_GreatPool/Meshes/T_FX_GreatPool_Skull_FX
  }
  n34 = Time()
  n35 = ScalarParameter() {defaultValue=5, group=Timing, parameterName=TimingSpeed}
  n36 = ScalarParameter() {defaultValue=5, group=Timing, parameterName=TimingOffset}
  n37 = ComponentMask(None=TextureCoordinate()) {g=true}
  n38 = NamedRerouteUsage()
  n39 = Add(A=Add(A=Multiply(A=@n34, B=@n35), B=Multiply(A=@n13.R, B=@n36) {constB=8}), B=Add(A=OneMinus(None=@n37), B=OneMinus(None=@n38)))
  n40 = ScalarParameter() {defaultValue=0.1, group=Timing, parameterName=LengthOfReveal}
  n41 = Multiply(A=TextureSample(UVs=Multiply(A=@n39, B=@n40)) {texture=/Game/Effects/Fort_Effects/Textures/Grad/T_FX_LinearSweep_Gradient}.G)
  n42 = Max(A=Max(A=@n33.R, B=Multiply(A=@n33.G, B=@n41) {constB=0}), B=@n33.B)
  n43 = NamedRerouteUsage()
  NamedRerouteDeclaration(None=Max(A=Max(A=@n42, B=@n43), B=NamedRerouteUsage())) {name=LightningDistanceField, nodeColor=<rgba 0.875 0 1 1>}
  n44 = NamedRerouteUsage()
  n45 = NamedRerouteUsage()
  NamedRerouteDeclaration(None=Multiply(A=NamedRerouteUsage(), B=LinearInterpolate(A=@n44, Alpha=@n45))) {name=ModifiedSDF, nodeColor=<rgba 0.5703 1 0 1>}
  NamedRerouteDeclaration(None=@n18.Linear Distance) {name=DistanceFromLightningOrigin, nodeColor=<rgba 0.8281 1 0 1>}
  n46 = Normalize(VectorInput=Subtract(
    A=CameraPositionWS()
    B=MaterialFunctionCall() {materialFunction=&ObjectPivotPoint}.Mesh Particle Pivot Location
   ))
  n47 = Normalize(VectorInput=Transform(None=Constant3Vector() {constant=<rgba 0 -1 0 0>}) {transformSourceType=TRANSFORMSOURCE_Instance})
  Saturate(None=Power(Base=DotProduct(A=@n46, B=@n47), Exp=ScalarParameter() {defaultValue=10, group=Utility, parameterName=2SidedSensitivity}) {constExponent=20})
  VertexInterpolator(VS=TextureCoordinate())
  WorldPosition()
 ]
}
