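# Niagara skydiving contrail for the GreatPool cosmetic, trailing audio-reactive electricity from the character.
# System scripts derive diving, audio and direction values, including drum kick, intensity and brightness, which drive emitter spawn rates, scale and glow.
# Bolt, coil and main electricity mesh emitters and arc sprites split particles across four sampled bones or sockets, with meshes oriented along computed vectors and animated through dynamic material parameters.
# Several emitters, including debug and spline bolts, are disabled.
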NiagaraSystem /BRCosmetics/Athena/Effects/Particles/Skydiving/NS_Contrail_GreatPool/NS_Contrail_GreatPool
dependencies {
 usedDataInterfaces [
  &NiagaraDataInterfaceCurve
  &NiagaraDataInterfaceCamera
  &NiagaraDataInterfaceVectorCurve
  &NiagaraDataInterfaceColorCurve
  &NiagaraDataInterfaceArrayFloat
  &NiagaraDataInterfaceArrayFloat3
  &NiagaraDataInterfaceMeshRendererInfo
  &NiagaraDataInterfaceSkeletalMesh
 ]
 usedDynamicInputs [
  &self:ShowAudioReactivity
  /Niagara/DynamicInputs/Lerp/SmoothLerpOverTime_Position
  &Abs_Float
  &ModuloFloat
  &ClampFloat
 ]
 usedModules [
  &self:SystemSpawnScript.SystemScriptSource.SystemScriptGraph.NiagaraNodeAssignment_0.SetVariables_#1
  &self:SystemSpawnScript.SystemScriptSource.SystemScriptGraph.NiagaraNodeAssignment_6.SetVariables_#2
  &SampleSkeletalMesh
  &self:SystemSpawnScript.SystemScriptSource.SystemScriptGraph.NiagaraNodeAssignment_8.SetVariables_#3
  &self:SystemSpawnScript.SystemScriptSource.SystemScriptGraph.NiagaraNodeAssignment_14.SetVariables_#4
  &self:SystemSpawnScript.SystemScriptSource.SystemScriptGraph.NiagaraNodeAssignment_13.SetVariables_#5
  &self:SystemSpawnScript.SystemScriptSource.SystemScriptGraph.NiagaraNodeAssignment_21.SetVariables_#6
  &self:SystemSpawnScript.SystemScriptSource.SystemScriptGraph.NiagaraNodeAssignment_7.SetVariables_#7
  &SystemState
  /Game/Effects/Niagara/Modules/Skydiving/FN_DivingAttributes/v6/FN_DivingAttributes
  &self:SystemSpawnScript.SystemScriptSource.SystemScriptGraph.NiagaraNodeAssignment_24.SetVariables_#8
  &self:SystemSpawnScript.SystemScriptSource.SystemScriptGraph.NiagaraNodeAssignment_15.SetVariables_#9
  &self:SystemSpawnScript.SystemScriptSource.SystemScriptGraph.NiagaraNodeAssignment_16.SetVariables_#10
  &self:SystemSpawnScript.SystemScriptSource.SystemScriptGraph.NiagaraNodeAssignment_4.SetVariables_#11
  &self:SystemSpawnScript.SystemScriptSource.SystemScriptGraph.NiagaraNodeAssignment_9.SetVariables_#12
  &InterpolateOverTime
  &self:SystemSpawnScript.SystemScriptSource.SystemScriptGraph.NiagaraNodeAssignment_10.SetVariables_#13
  &self:SystemSpawnScript.SystemScriptSource.SystemScriptGraph.NiagaraNodeAssignment_5.SetVariables_#14
  &self:SystemSpawnScript.SystemScriptSource.SystemScriptGraph.NiagaraNodeAssignment_1.SetVariables_#15
  &FN_Integrator
  &Timeline
  &self:SystemSpawnScript.SystemScriptSource.SystemScriptGraph.NiagaraNodeAssignment_2.SetVariables_#16
  &self:SystemSpawnScript.SystemScriptSource.SystemScriptGraph.NiagaraNodeAssignment_3.SetVariables_#17
  &DoOnce
  &self:SystemSpawnScript.SystemScriptSource.SystemScriptGraph.NiagaraNodeAssignment_12.SetVariables_#18
  &self:SystemSpawnScript.SystemScriptSource.SystemScriptGraph.NiagaraNodeAssignment_11.SetVariables_#19
  &self:SystemSpawnScript.SystemScriptSource.SystemScriptGraph.NiagaraNodeAssignment_17.SetVariables_#20
  &self:SystemSpawnScript.SystemScriptSource.SystemScriptGraph.NiagaraNodeAssignment_18.SetVariables_#21
  &self:SystemSpawnScript.SystemScriptSource.SystemScriptGraph.NiagaraNodeAssignment_23.SetVariables_#22
  &self:SystemSpawnScript.SystemScriptSource.SystemScriptGraph.NiagaraNodeAssignment_19.SetVariables_#23
  &self:SystemSpawnScript.SystemScriptSource.SystemScriptGraph.NiagaraNodeAssignment_20.SetVariables_#24
  /Game/Effects/Niagara/Modules/Logic/UpdateValueStateChange
  &self:SystemSpawnScript.SystemScriptSource.SystemScriptGraph.NiagaraNodeAssignment_22.SetVariables_#25
 ]
}
emitters {
 Bolt_Meshes {
  emitterUpdateScript [
   &EmitterState(Life Cycle Mode=::NewEnumerator1, Loop Behavior=::NewEnumerator1, Loop Duration Mode=::NewEnumerator1, UseLoopDelay=false)
   &SpawnRate(
    SpawnGroup=1
    SpawnRate=&Multiply_Float(
     A=&FloatFromCurve(
      FloatCurve=DataInterface() {propertyValues {Curve {keys [<at 0 0 RCIM_Cubic>, {arriveTangent=16.6667, interpMode=RCIM_Cubic, leaveTangent=16.6667, time=0.3, value=1}, <at 0.9 15 RCIM_Cubic>]}}}
      CurveIndex=$System.Audio_DrumKick
      Scale Curve=0.75
     )
     B=&FN_FrontEndFloatModifier(Gameplay=1, Front End=0.65)
    )
    Use Spawn Probability=false
   ) {module=SpawnRate001}
   &self:SpawnOnConditionMet(
    Spawn Count=12
    SpawnCondition:NiagaraBool=&SetBoolByFloatComparison(A=$System.Audio_DrumKick, B=0.99)
    Use Spawn Probability=false
    UseLoopCountLimit=false
   ) {enabled=false, module=SpawnOnConditionMet}
  ]
  particleSpawnScript [
   /SetVariables(Particles.VisibilityTag=&SpawnGroup()) {isSetParameters=true}
   &InitializeParticle(
    Color=&MaskLinearColorBySpawnGroup(
     Masked Linear Color=<rgba 0.5 0.5 0.5 1>
     Passthrough Linear Color=&LerpLinearColors(
      StartColor=<rgba 0.4 0.4 0.4 1>
      EndColor=<rgba 2 2 2 1>
      LerpFactor=/Niagara/DynamicInputs/Math/Floor_Float(Float:NiagaraFloat=&Multiply_Float(A=&RandomRangeFloat(Minimum=0, Maximum=1), B=1.5))
     )
    )
    Lifetime=&RandomRangeFloat(Minimum=0.2, Maximum=0.4)
    Mesh Scale Mode=::NewEnumerator3
    Mesh Uniform Scale=&Multiply_Float(A=&RandomRangeFloat(Minimum=0.7, Maximum=1.2), B=0.7)
    Position Offset=<15 0 0>
    UsePositionOffset=true
    Write Material Random=true
   )
   &InitialMeshOrientation(
    Mesh Orientation Mode=::NewEnumerator7
    Rotation=&RandomRangeVector(Minimum=<0 -0.25 0>, Maximum=<0 -0.25 0>)
    Use Rotation Vector=false
   )
   &ShapeLocation(MaskSpawnGroup=true, Override Seed=false, Sphere Radius=0, UseOwnerScale=true) {enabled=false}
   &ShapeLocation(MaskSpawnGroup=true, Override Seed=false, Spawn Group Mask=1, Sphere Radius=9, UseOwnerScale=true) {enabled=false, module=ShapeLocation001}
   &PartitionParticles(Number Of Partitions=4)
  ]
  particleUpdateScript [
   &ParticleState()
   &ScaleMeshSize(Scale Factor=&RandomRangeVector(Minimum=<0.3 0.3 1>, Maximum=<0.3 0.3 1>)) {module=ScaleMeshSize001}
   &ScaleMeshSize(Scale Factor=&MaskVectorBySpawnGroup(
     Masked Vector=<1 1 1>
     Passthrough Vector=&MakeVector(
      X=&FloatFromCurve(
       FloatCurve=DataInterface() {propertyValues {Curve {keys [<at 0 0>, {arriveTangent=0.0951, leaveTangent=-0.1745, time=0.15, value=1}]}}}
       CurveIndex=$Emitter.Age
      )
      Y=&FloatFromCurve(
       FloatCurve=DataInterface() {propertyValues {Curve {keys [<at 0 0>, {arriveTangent=0.0951, leaveTangent=-0.1745, time=0.15, value=1}]}}}
       CurveIndex=$Emitter.Age
      )
      Z=&FloatFromCurve(
       FloatCurve=DataInterface() {propertyValues {Curve {keys [
           {interpMode=RCIM_Cubic, leaveTangent=0.0966, tangentMode=RCTM_Break, time=0, value=1}
           {arriveTangent=0.0951, interpMode=RCIM_Cubic, leaveTangent=-0.1745, tangentMode=RCTM_Break, time=1, value=1.07}
          ]}}}
       Scale Curve=&Lerp_Float(A=1, B=1.25, Alpha=$System.Dive)
      )
     )
    ))
   &SkeletalMeshLocation(
    Bone / Socket Index=$Particles.PartitionParticles.Partition
    Bone Sampling Mode=::NewEnumerator4
    MaskSpawnGroup=false
    Orientation Sampling=::NewEnumerator1
    Skeletal Mesh=$System.Skeletal Mesh
   ) {module=SkeletalMeshLocation001}
   &DynamicMaterialParameters(
    Index 0 Param 4=&MaskFloatBySpawnGroup(
     Masked Float=1
     Passthrough Float=&SelectFloatFromArray(
      Direct Array Index=&ModuloInt(A=&ReturnParticleIDAttributes(), B=3)
      Float Selection Array=DataInterface() {propertyValues {FloatData [3,0,10,3]}}
     )
    )
    Index 1 Param 1=1
    Index 1 Param 2=2
    Index 1 Param 3=&MaskFloatBySpawnGroup(
     Masked Float=-1
     Passthrough Float=&FloatFromCurve(FloatCurve=DataInterface() {propertyValues {Curve {keys [{interpMode=RCIM_Cubic, tangentMode=RCTM_Break, time=0.3, value=-0.5}, <at 1 0.1>]}}})
    )
    Index 2 Param 1=&FloatFromCurve(
     FloatCurve=DataInterface() {propertyValues {Curve {keys [
         {interpMode=RCIM_Cubic, tangentMode=RCTM_Break, time=0, value=0}
         {interpMode=RCIM_Cubic, tangentMode=RCTM_Break, time=0.2, value=1}
         {interpMode=RCIM_Cubic, tangentMode=RCTM_Break, time=1, value=0}
        ]}}}
     Scale Curve=0.02
    )
    Index 2 Param 2=&MaskFloatBySpawnGroup(
     Masked Float=0
     Passthrough Float=&FloatFromCurve(
      FloatCurve=DataInterface() {propertyValues {Curve {keys [
          {interpMode=RCIM_Cubic, tangentMode=RCTM_Break, time=0.7, value=0}
          {arriveTangent=-0.0859, interpMode=RCIM_Cubic, tangentMode=RCTM_Break, time=1, value=-1}
         ]}}}
      Scale Curve=50
     )
    )
    Index 2 Param 3=&MaskFloatBySpawnGroup(
     Masked Float=0
     Passthrough Float=&Waveform(
      Global Input Position=&RandomRangeFloat(Minimum=0, Maximum=155)
      Global Phase=$Particles.Age
      "Waveform [1]"=::NewEnumerator12
      "[1] Phase Scale"=2
      "[1] Frequency"=1
      "[1] Amplitude Min/Max"=<0 1>
     )
    )
    Index 2 Param 4=&MaskFloatBySpawnGroup(Masked Float=0, Passthrough Float=25)
    Index 3 Param 3=2
    Index 3 Param 4=-1
    Module.Index0Param0Write=true
    Module.Index0Param1Write=true
    Module.Index0Param2Write=true
    Module.Index0Param3Write=true
    Module.Index1Param0Write=true
    Module.Index1Param1Write=true
    Module.Index1Param2Write=true
    Module.Index1Param3Write=true
    Module.Index2Param0Write=true
    Module.Index2Param1Write=true
    Module.Index2Param2Write=true
    Module.Index2Param3Write=true
    Module.Index3Param0Write=true
    Module.Index3Param1Write=true
    Module.Index3Param2Write=true
    Module.Index3Param3Write=true
    Write Parameter Index 1=true
    Write Parameter Index 2=true
    Write Parameter Index 3=true
   )
   &ScaleColor(
    Scale Alpha=&FloatFromCurve(FloatCurve=DataInterface() {propertyValues {Curve {keys [
         {tangentMode=RCTM_User, time=0, value=1}
         {arriveTangent=-0.0405, interpMode=RCIM_Cubic, leaveTangent=-0.0405, tangentMode=RCTM_User, time=0.7009, value=1}
         {arriveTangent=-8.0089, interpMode=RCIM_Cubic, leaveTangent=-8.0089, tangentMode=RCTM_User, time=1, value=0}
        ]}}})
    Scale RGB=&Multiply_VectorByFloat(
     Vector=&MaskVectorBySpawnGroup(
      Masked Vector=&VectorFromFloat(Value=&RandomRangeFloat(Minimum=3, Maximum=4, Evaluation Type=::NewEnumerator1))
      Passthrough Vector=&Lerp_Vector(
       A=&VectorFromFloat(Value=&FloatFromCurve(FloatCurve=DataInterface() {propertyValues {Curve {keys [<at 0.2771 1>,<at 0.3786 1>]}}}, Scale Curve=3))
       B=&VectorFromFloat(Value=&RandomRangeFloat(Minimum=1, Maximum=1, Evaluation Type=::NewEnumerator1))
       Alpha=&FloatFromCurve(FloatCurve=DataInterface() {propertyValues {Curve {keys [
            {interpMode=RCIM_Cubic, leaveTangent=2.5294, tangentMode=RCTM_Break, time=0, value=0}
            {arriveTangent=-0.2166, interpMode=RCIM_Cubic, tangentMode=RCTM_Break, time=1, value=1}
           ]}}})
      )
     )
     Float=&Lerp_Float(
      A=0.75
      B=1.5
      Alpha=&FloatFromCurve(
       FloatCurve=DataInterface() {propertyValues {Curve {keys [
           {interpMode=RCIM_Cubic, tangentMode=RCTM_User, time=0, value=0}
           {arriveTangent=1, interpMode=RCIM_Cubic, leaveTangent=1, time=0.5, value=0.125}
           {interpMode=RCIM_Cubic, tangentMode=RCTM_User, time=1, value=1}
          ]}}}
       CurveIndex=$System.AudioBrightnessMultiplier
      )
     )
    )
    ScaleA=true
    ScaleRGB=true
   )
   /SetVariables(Particles.MeshOrientation=&OrientMeshToVector(Look At Direction=$System.SysLightningFaceDir, Reference Axis Vector=<0 0 1>)) {isSetParameters=true}
   /SetVariables(Particles.MeshOrientation=&OrientMeshToVector(Look At Direction=$System.SysLightningFaceDir, Reference Axis Vector=<0 0 1>)) {enabled=false, isSetParameters=true}
   /SetVariables(Particles.MeshOrientation2:Quat4f=&MultiplyQuaternion(
     Quaternion A=$Particles.MeshOrientation
     Quaternion B=&MakeQuatFromAxisAngle(Axis=<0 0 1>, AngleInDegrees=&RandomRangeFloat(Minimum=0, Maximum=360))
    )) {isSetParameters=true}
   /SetVariables(Particles.MeshOrientation2:Quat4f=&MultiplyQuaternion(
     Quaternion A=$Particles.MeshOrientation2
     Quaternion B=&MakeQuatFromAxisAngle(
      Axis=<1 0 0>
      AngleInDegrees=&Multiply_Float(A=$System.LightningAngle, B=&RandomRangeFloat(Minimum=-10, Maximum=10))
     )
    )) {isSetParameters=true}
  ]
  rendererClasses [&NiagaraMeshRendererProperties]
  renderers [
   {rendererClass=&NiagaraMeshRendererProperties, rendererIndex=0}
   {rendererClass=&NiagaraMeshRendererProperties, rendererIndex=1}
  ]
 }
 ElectricCoils {
  emitterUpdateScript [
   &EmitterState(Life Cycle Mode=::NewEnumerator1, Loop Duration=1, UseLoopDelay=false)
   &SpawnRate(
    SpawnRate=&FloatFromCurve(
     FloatCurve=DataInterface() {propertyValues {Curve {keys [<at 0 1>, {arriveTangent=6.8, leaveTangent=6.8, time=0.4, value=1.1}, <at 0.9 8 RCIM_Cubic>]}}}
     CurveIndex=$System.Audio_DrumKick
    )
    Use Spawn Probability=false
   )
  ]
  particleSpawnScript [
   &InitializeParticle(
    Color=$System.MainColor
    Lifetime=&RandomRangeFloat(Minimum=0.25, Maximum=0.55)
    Mesh Scale Mode=::NewEnumerator3
    Mesh Uniform Scale=1.1
    Sprite Size Mode=::NewEnumerator3
    Uniform Sprite Size=50
    UsePositionOffset=false
    Write Material Random=true
   )
   &PartitionParticles(Number Of Partitions=4, Particles Per Partition=1, Partition Mode:ENiagara_PartitionMode=::NewEnumerator1)
   &SkeletalMeshLocation(
    Bone / Socket Index=$Particles.PartitionParticles.Partition
    Bone Sampling Mode=::NewEnumerator4
    MaskSpawnGroup=false
    Orientation Sampling=::NewEnumerator1
    Skeletal Mesh=$System.Skeletal Mesh
   ) {module=SkeletalMeshLocation001}
   &InitialMeshOrientation(
    Mesh Orientation Mode=::NewEnumerator7
    Rotation=&MakeVector(X=&RandomRangeFloat(Minimum=0, Maximum=360), Y=0, Z=0)
    Use Rotation Vector=true
   )
  ]
  particleUpdateScript [
   &ParticleState()
   /SetVariables(
    Particles.LightningLength:NiagaraFloat=&FloatFromCurve(FloatCurve=DataInterface() {propertyValues {Curve {keys [{arriveTangent=1.2262, interpMode=RCIM_Cubic, leaveTangent=1.2262, tangentMode=RCTM_User, time=0, value=0.5}, <at 1 1 RCIM_Cubic>]}}})
    Particles.LightningShapeExpand:NiagaraFloat=&FloatFromCurve(
     FloatCurve=DataInterface() {propertyValues {Curve {keys [
         {arriveTangent=4.1785, interpMode=RCIM_Cubic, leaveTangent=4.1785, tangentMode=RCTM_User, time=0, value=0}
         {arriveTangent=1.1179, interpMode=RCIM_Cubic, leaveTangent=1.1179, tangentMode=RCTM_User, time=0.2042, value=0.6228}
         {arriveTangent=0.0224, interpMode=RCIM_Cubic, leaveTangent=0.0224, tangentMode=RCTM_User, time=1, value=1}
        ]}}}
     Scale Curve=&Lerp_Float(A=0.5, B=0.5, Alpha=$System.Dive)
    )
   ) {isSetParameters=true, module=SetVariables_#26}
   &ScaleMeshSize(Scale Factor=&MakeVector(X=$Particles.LightningShapeExpand, Y=$Particles.LightningShapeExpand, Z=$Particles.LightningLength))
   &SkeletalMeshLocation(
    Bone / Socket Index=$Particles.PartitionParticles.Partition
    Bone Sampling Mode=::NewEnumerator4
    MaskSpawnGroup=false
    Orientation Sampling=::NewEnumerator1
    Skeletal Mesh=$System.Skeletal Mesh
   )
   &SkeletalMeshLocation(
    Bone / Socket Index=&Add_Integer(A=$Particles.PartitionParticles.Partition, B=5)
    Bone Sampling Mode=::NewEnumerator4
    MaskSpawnGroup=false
    Orientation Sampling=::NewEnumerator1
    Position Sampling=::NewEnumerator2
    Skeletal Mesh=$System.Skeletal Mesh
   ) {module=SkeletalMeshLocation002}
   /SetVariables(Particles.MeshOrientation=&MultiplyQuaternion(
     Quaternion A=&OrientMeshToVector(
      Look At Direction=&NormalizeVector(
       Vector To Normalize=&Subtract_Vector(A=$Output.SkeletalMeshLocation002.SampledPositionWithoutOffset, B=$Output.SkeletalMeshLocation.SampledPositionWithoutOffset)
       Safe Normalize=true
       Safe Fallback Threshold=0
       Safe Fallback Vector=<0 0 0.001>
      )
      Reference Axis Vector=<1 0 0>
     )
     Quaternion B=&MakeQuatFromAxisAngle(Axis=<1 0 0>, AngleInDegrees=&Multiply_Float(A=$Particles.NormalizedAge, B=-150))
    )) {isSetParameters=true, module=SetVariables_#27}
   &DynamicMaterialParameters(
    Index 0 Param 1=&FloatFromCurve(FloatCurve=DataInterface() {propertyValues {Curve {keys [<at 0 1>,<at 1 0.5>]}}}, Scale Curve=1.25)
    Index 0 Param 3=&RandomRangeFloat(Minimum=0, Maximum=22)
    Index 0 Param 4=0.45
    Index 1 Param 1=&FloatFromCurve(
     FloatCurve=DataInterface() {propertyValues {Curve {keys [
         <at 0 0>
         {arriveTangent=1, interpMode=RCIM_Cubic, leaveTangent=1, time=0.3, value=0}
         {interpMode=RCIM_Cubic, leaveTangent=-8.0089, time=1, value=1}
        ]}}}
     Scale Curve=0.1
    )
    Index 1 Param 2=1
    Index 1 Param 3=&FloatFromCurve(FloatCurve=DataInterface() {propertyValues {Curve {keys [<at 0.5 1 RCIM_Cubic>,<at 1 0 RCIM_Cubic>]}}}, Scale Curve=2)
    Index 2 Param 1=&FloatFromCurve(
     FloatCurve=DataInterface() {propertyValues {Curve {keys [{tangentMode=RCTM_User, time=0, value=0}, {tangentMode=RCTM_User, time=1, value=1}]}}}
     Scale Curve=0
    )
    Index 2 Param 2=2
    Module.Index0Param0Write=true
    Module.Index0Param1Write=true
    Module.Index0Param2Write=true
    Module.Index0Param3Write=true
    Module.Index1Param0Write=true
    Module.Index1Param1Write=true
    Module.Index1Param2Write=true
    Module.Index1Param3Write=true
    Module.Index2Param0Write=true
    Module.Index2Param1Write=true
    Module.Index2Param2Write=true
    Module.Index2Param3Write=true
    Write Parameter Index 1=true
    Write Parameter Index 2=true
   )
   &ScaleColor(
    Scale Alpha=&FloatFromCurve(FloatCurve=DataInterface() {propertyValues {Curve {keys [
         {tangentMode=RCTM_User, time=0, value=1}
         {arriveTangent=-0.0405, interpMode=RCIM_Cubic, leaveTangent=-0.0405, tangentMode=RCTM_User, time=0.7009, value=1}
         {arriveTangent=-8.0089, interpMode=RCIM_Cubic, leaveTangent=-8.0089, tangentMode=RCTM_User, time=1, value=0}
        ]}}})
    Scale RGB=&VectorFromFloat(Value=&FloatFromCurve(
      FloatCurve=DataInterface() {propertyValues {Curve {keys [
          {arriveTangent=9.0725, interpMode=RCIM_Cubic, time=0, value=0}
          <at 0.1408 2 RCIM_Cubic>
          {interpMode=RCIM_Cubic, leaveTangent=-0.0405, time=1, value=0}
         ]}}}
      Scale Curve=&Lerp_Float(A=4, B=7, Alpha=$System.AudioBrightnessMultiplier)
     ))
    ScaleA=true
    ScaleRGB=true
   ) {module=ScaleColor001}
   &UpdateMeshOrientation(Rotation Rate=0) {enabled=false, module=UpdateMeshOrientation002}
   &CameraOffset(Camera Offset Amount=5)
  ]
  rendererClasses [&NiagaraMeshRendererProperties]
  renderers [{rendererClass=&NiagaraMeshRendererProperties, rendererIndex=0}]
 }
 Empty {
  bEnabled=false
  emitterUpdateScript [
   &EmitterState()
   &SpawnBurst_Instantaneous(Spawn Count=4, Use Spawn Probability=false, UseLoopCountLimit=false)
  ]
  particleSpawnScript [
   &InitializeParticle(
    Color Mode=::NewEnumerator0
    Lifetime=5
    Sprite Size Mode=::NewEnumerator3
    Uniform Sprite Size=36.6616
    UsePositionOffset=false
    Write Material Random=true
   )
   /Niagara/Modules/Beams/SpawnBeam() {enabled=false}
   &PartitionParticles(Number Of Partitions=4)
  ]
  particleUpdateScript [
   &ParticleState()
   /SetVariables(Particles.Color=&MultiplyLinearColor(
     Linear Color=$System.MainColor
     Scale Factor=&MakeFloat4FromVectorAndFloat(
      "Vector (RGB)"=&VectorFromFloat(Value=&Waveform(
        Global Phase=&Add_Float(A=$Particles.Age, B=&RandomRangeFloat(Minimum=0, Maximum=3))
        "Waveform [1]"=::NewEnumerator0
        "[1] Frequency"=5
        "[1] Amplitude Min/Max"=<0.7 1.1>
       ))
      "Float (Alpha)"=0.5
     )
    )) {isSetParameters=true}
   &SkeletalMeshLocation(
    Bone / Socket Index=$Particles.PartitionParticles.Partition
    Bone Sampling Mode=::NewEnumerator4
    Inherit Velocity=false
    Interpolate Mesh Sampling=false
    MaskSpawnGroup=false
    Sampled Position Offset=&SelectVectorFromArray(
     Direct Array Index=$Particles.PartitionParticles.Partition
     Vector Selection Array=DataInterface() {propertyValues {FloatData [<10 0 0>, <-10 0 0>, <-7 5 0>, <7 -5 0>]}}
    )
    Sampled Position Offset Coordinate Space:ENiagaraMeshTransforms=::NewEnumerator3
    Skeletal Mesh=$System.Skeletal Mesh
   )
   /SetVariables(
    Particles.FlareColor=&MultiplyLinearColor(Linear Color=$Particles.Color, Scale Factor=<0.5 0.5 0.5 0.5>)
    Particles.FlareSize:Vector2f=&Multiply_Vector2D_ByFloat(
     Float=&Waveform(
      Global Phase=&Add_Float(A=$Particles.Age, B=&RandomRangeFloat(Minimum=0, Maximum=3))
      "Waveform [1]"=::NewEnumerator0
      "[1] Frequency"=5
      "[1] Amplitude Min/Max"=<0.7 1.1>
     )
     Vector2D=<128 64>
    )
   ) {isSetParameters=true}
   &CameraOffset(Camera Offset Amount=20)
  ]
  rendererClasses [&NiagaraSpriteRendererProperties]
  renderers [
   {rendererClass=&NiagaraSpriteRendererProperties}
   {rendererClass=&NiagaraSpriteRendererProperties, rendererIndex=1}
  ]
 }
 EnergySplineBolts {
  bEnabled=false
  emitterUpdateScript [
   &EmitterState()
   /SetVariables(Emitter.V_Scale_Random:NiagaraFloat=&RandomRangeFloat(Minimum=0.7, Maximum=1.1)) {isSetParameters=true}
   &SpawnRate(
    SpawnGroup:NiagaraInt32=&MakeCustomIntFromBool(Bool=$User.TargetIsLaserGrid, True INT=1, False INT=0)
    SpawnRate=&FloatFromCurve(FloatCurve=DataInterface() {propertyValues {Curve {keys [<at 0 1>,<at 1 5>]}}}, CurveIndex=$System.AudioIntensity)
    Use Spawn Probability=false
   )
  ]
  particleSpawnScript [
   &InitializeParticle(
    Color=$System.MainColor
    Lifetime Max=1.3429
    Lifetime Min=0.628
    Lifetime Mode=::NewEnumerator1
    Mesh Scale Mode=::NewEnumerator1
    Mesh Uniform Scale Max=0.125
    Mesh Uniform Scale Min=0.1
    UsePositionOffset=true
    Write Material Random=true
   )
   &InitialMeshOrientation(Mesh Orientation Mode=::NewEnumerator6, Orientation Quaternion:Quat4f=$System.SkullRotation, Rotation=<0 0 0.25>, Use Rotation Vector=true) {enabled=false, module=InitialMeshOrientation002}
   &InitialMeshOrientation(
    Mesh Orientation Mode=::NewEnumerator7
    Rotation=&RandomRangeVector(Minimum=<0 0 0>, Maximum=<0 0 250>)
    Use Rotation Vector=true
   ) {module=InitialMeshOrientation001}
   &InitialMeshOrientation(
    Mesh Orientation Mode=::NewEnumerator6
    Orientation Quaternion:Quat4f=$System.SkullRotation
    Rotation=&MakeVector(X=&RandomRangeFloat(Minimum=0, Maximum=360), Y=0, Z=0)
    Use Rotation Vector=true
   )
   &CameraOffset() {enabled=false}
   &ShapeLocation(MaskSpawnGroup=false, Override Seed=false, Sphere Radius=0, UseOwnerScale=true) {enabled=false}
  ]
  particleUpdateScript [
   /SetVariables(Particles.Flicker:NiagaraFloat=&RandomRangeFloat(Minimum=-0.1, Maximum=1, Evaluation Type=::NewEnumerator1)) {isSetParameters=true}
   /SetVariables(Particles.Position=$System.SkullPosition) {isSetParameters=true}
   &ParticleState()
   &ScaleMeshSize(Scale Factor=&VectorFromFloat(Value=&Multiply_Float(A=10, B=1.5)))
   &ScaleMeshSize(Scale Factor=&VectorFromCurve(
     VectorCurve=DataInterface() {propertyValues {
       XCurve {keys [<at 0 1>]}
       YCurve {keys [
         {arriveTangent=4.1785, interpMode=RCIM_Cubic, leaveTangent=4.1785, tangentMode=RCTM_User, time=0, value=0}
         {arriveTangent=1.1179, interpMode=RCIM_Cubic, leaveTangent=1.1179, tangentMode=RCTM_User, time=0.2042, value=0.6228}
         {arriveTangent=0.0224, interpMode=RCIM_Cubic, leaveTangent=0.0224, tangentMode=RCTM_User, time=1, value=1}
        ]}
       ZCurve {keys [
         {interpMode=RCIM_Cubic, leaveTangent=1.5389, tangentMode=RCTM_Break, time=0, value=0.8}
         {interpMode=RCIM_Cubic, tangentMode=RCTM_Break, time=1, value=1.5}
        ]}
      }}
     Scale Curve=<1 2 1.5>
    )) {module=ScaleMeshSize001}
   &DynamicMaterialParameters(
    Index 0 Param 1=1
    Index 0 Param 2=5
    Index 0 Param 4=&RandomRangeFloat(Minimum=0.1, Maximum=0.5)
    Index 1 Param 1=&Multiply_Float(
     A=&Multiply_Float(
      A=&Multiply_Float(A=1.25, B=1.3)
      B=&RandomRangeFloat(Minimum=0.2, Maximum=1)
     )
     B=3.5
    )
    Index 1 Param 2=1
    Index 1 Param 3=&RandomRangeFloat(Minimum=0.8, Maximum=1.8)
    Index 1 Param 4=1
    Index 2 Param 1=&FloatFromCurve(FloatCurve=DataInterface() {propertyValues {Curve {keys [
         <at -0.0019 0 RCIM_Cubic>
         {arriveTangent=1.1676, interpMode=RCIM_Cubic, leaveTangent=1.1676, tangentMode=RCTM_User, time=0.6656, value=0.5892}
         {arriveTangent=1.2212, interpMode=RCIM_Cubic, leaveTangent=1.2212, time=0.9445, value=0.9824}
         {arriveTangent=0.2189, interpMode=RCIM_Cubic, leaveTangent=0.2189, tangentMode=RCTM_User, time=1.0019, value=1}
        ]}}})
    Module.Param0WriteEnabled=true
    Module.Param0WriteEnabled1=true
    Module.Param0WriteEnabled2=true
    Module.Param1WriteEnabled=true
    Module.Param1WriteEnabled1=true
    Module.Param1WriteEnabled2=true
    Module.Param2WriteEnabled=true
    Module.Param2WriteEnabled1=true
    Module.Param2WriteEnabled2=true
    Module.Param3WriteEnabled=true
    Module.Param3WriteEnabled1=true
    Module.Param3WriteEnabled2=true
    Write Parameter Index 1=true
    Write Parameter Index 2=true
   )
   &ScaleColor(
    Linear Color Curve=DataInterface() {propertyValues {
      AlphaCurve {keys [
        {interpMode=RCIM_Cubic, tangentMode=RCTM_User, time=0, value=1}
        {arriveTangent=-1, interpMode=RCIM_Cubic, leaveTangent=-1, time=0.3555, value=0.2253}
        {interpMode=RCIM_Cubic, tangentMode=RCTM_User, time=1, value=0}
       ]}
      BlueCurve {keys [<at 0 10>,<at 0.4774 10>,<at 0.9341 1>]}
      GreenCurve {keys [<at 0 10>,<at 0.4774 10>,<at 0.9341 0.1667>]}
      RedCurve {keys [<at 0 10>,<at 0.4774 10>,<at 0.9341 0>]}
     }}
    Scale Mode=::NewEnumerator2
   )
   &ScaleColor(
    Scale Alpha=&Lerp_Float(A=0.3, B=1, Alpha=$Particles.Flicker)
    Scale RGB=&VectorFromFloat(Value=&Multiply_Float(
      A=&Lerp_Float(A=1, B=2, Alpha=$Particles.Flicker)
      B=&FloatFromCurve(FloatCurve=DataInterface() {propertyValues {Curve {keys [{interpMode=RCIM_Cubic, tangentMode=RCTM_User, time=0, value=1}, {interpMode=RCIM_Cubic, tangentMode=RCTM_User, time=1, value=0}]}}})
     ))
    ScaleA=true
    ScaleRGB=true
   ) {module=ScaleColor001}
  ]
  rendererClasses [&NiagaraMeshRendererProperties]
  renderers [
   {rendererClass=&NiagaraMeshRendererProperties, rendererIndex=0}
   {rendererClass=&NiagaraMeshRendererProperties, rendererIndex=1}
  ]
 }
 FN_Debug {
  bEnabled=false
  emitterUpdateScript [
   &EmitterState()
   &SpawnBurst_Instantaneous(Spawn Count=1, Use Spawn Probability=false)
  ]
  particleSpawnScript [/Niagara/Modules/Spawn/Initialization/InitializeParticle(
    Lifetime=5
    Module.Write Color=true
    Module.Write Lifetime=true
    Module.Write Mass=true
    Module.Write Material Random=true
    Module.Write Position=false
    Module.Write Scale=false
    Module.Write SpriteRotation=false
    Module.Write SpriteSize=true
    Position=&SimulationPosition()
    Sprite Size=<100 100>
   )]
  particleUpdateScript [
   &ParticleState(Kill Particles When Lifetime Has Elapsed=false)
   /SetVariables(Particles.Debug:NiagaraFloat=&RemapRange(
     Input Value To Remap=&MakeFloatFromLinearColor(LinearColor=$NPC.NPC_Audio_Reactive_Parameters.Spectrum_Music, Channel:ENiagaraLinearColor_Channels=::NewEnumerator0)
     Input Min=0
     Input Max=1
     Output Min=0
     Output Max=1
    )) {isSetParameters=true}
   &DynamicMaterialParameters(
    Index 0 Param 1=$Particles.Debug
    Module.Param0WriteEnabled=true
    Module.Param1WriteEnabled=true
    Module.Param2WriteEnabled=true
    Module.Param3WriteEnabled=true
   )
   &CameraOffset(Camera Offset Amount=100)
   &Color(Color=&MakeCustomLinearColorFromBool(
     Bool=&SetBoolByFloatComparison(A=$System.SideVelocity, B=0)
     True Color=<rgba 0 1 0.1993 1>
     False Color=<rgba 1 0 0.0842 1>
    ))
  ]
  rendererClasses [&NiagaraSpriteRendererProperties]
  renderers [{rendererClass=&NiagaraSpriteRendererProperties}]
 }
 Lightning_Arcs {
  emitterUpdateScript [
   &EmitterState()
   &SpawnRate(
    SpawnGroup=1
    SpawnRate=&Multiply_FloatByInt(
     Float=&FloatFromCurve(
      FloatCurve=DataInterface() {propertyValues {Curve {keys [{interpMode=RCIM_Cubic, tangentMode=RCTM_User, time=0, value=0}, {interpMode=RCIM_Cubic, tangentMode=RCTM_User, time=1, value=1}]}}}
      CurveIndex=$System.AudioIntensity
      Scale Curve=6
     )
     Integer=4
    )
    Use Spawn Probability=false
   ) {module=SpawnRate001}
  ]
  particleSpawnScript [/Niagara/Modules/Spawn/Initialization/V2/InitializeParticle(
    Color=&MultiplyLinearColor(
     Linear Color=$System.MainColor
     Scale Factor=&MakeFloat4FromVectorAndFloat("Vector (RGB)"=&VectorFromFloat(Value=3), "Float (Alpha)"=1)
    )
    Lifetime Max=0.2
    Lifetime Min=0.25
    Lifetime Mode=::NewEnumerator1
    Mass Mode=::NewEnumerator3
    Position Mode=::NewEnumerator0
    Sprite Rotation Mode=::NewEnumerator2
    Sprite Size=&Multiply_Vector2D_ByFloat(Float=&RandomRangeFloat(Minimum=10, Maximum=35), Vector2D=<1 1>)
    Sprite Size Mode=::NewEnumerator4
    Sprite UV Mode=::NewEnumerator1
    Write Material Random=true
   )]
  particleUpdateScript [
   &ParticleState()
   &SpriteRotationRate(Rotation Rate=&RandomRangeFloat(Minimum=-40, Maximum=50))
   /SetVariables(Particles.VisibilityTag=&MakeCustomIntFromBool(
     Bool=&SetBoolByFloatComparison(A=$System.CloudDot, B=0.5)
     True INT=&RandomRangeInt(Minimum=0, Maximum=2)
     False INT=&RandomRangeInt(Minimum=3, Maximum=5)
    )) {isSetParameters=true}
   &ScaleSpriteSize(Uniform Curve Sprite Scale=DataInterface() {propertyValues {Curve {keys [
        {
         arriveTangent=2.1133
         arriveTangentWeight=1.0818
         interpMode=RCIM_Cubic
         leaveTangent=2.1133
         leaveTangentWeight=1.0818
         tangentMode=RCTM_User
         time=0
         value=0
        }
        {arriveTangentWeight=0.8185, leaveTangentWeight=0.8185, time=1, value=1}
       ]}}})
   &ScaleSpriteSize(
    Scale Sprite Size Mode=::NewEnumerator0
    Uniform Scale Factor=&MaskFloatBySpawnGroup(
     Masked Float=1
     Passthrough Float=&FloatFromCurve(
      FloatCurve=DataInterface() {propertyValues {Curve {keys [{tangentMode=RCTM_User, time=-1, value=0.5}, {tangentMode=RCTM_User, time=-0.7, value=1}]}}}
      CurveIndex=&DotProduct(Vector A=<0 0 1>, Vector B=$System.CamForVec)
     )
     Spawn Group Mask=1
    )
   ) {module=ScaleSpriteSize001}
   &ScaleColor(
    Scale Alpha=&FloatFromCurve(FloatCurve=DataInterface() {propertyValues {Curve {keys [
         {arriveTangent=4.1298, interpMode=RCIM_Cubic, leaveTangent=4.1298, tangentMode=RCTM_User, time=0, value=0}
         {interpMode=RCIM_Cubic, tangentMode=RCTM_User, time=0.5, value=1}
         {arriveTangent=-5.4005, interpMode=RCIM_Cubic, leaveTangent=-5.4005, tangentMode=RCTM_User, time=1, value=0}
        ]}}})
    ScaleA=true
    ScaleRGB=true
   )
   &ShapeLocation(MaskSpawnGroup=false, Override Seed=false, Shape Origin=$System.Lighting Start, Sphere Radius=3, UseOwnerScale=true)
   &ScaleColor(
    Scale RGB=&VectorFromFloat(Value=&Lerp_Float(A=0.35, B=1, Alpha=&Sine(Normalized Angle=$Emitter.Age, Period=0.1, Scale=0.5, Bias=0.5)))
    ScaleA=true
    ScaleRGB=true
   ) {module=ScaleColor001}
   &SkeletalMeshLocation(
    Bone / Socket Index=&ModuloInt(A=&ReturnParticleIDAttributes(), B=4)
    Bone Sampling Mode=::NewEnumerator4
    MaskSpawnGroup=false
    Orientation Sampling=::NewEnumerator1
    Position Sampling=::NewEnumerator1
    Sampled Position Offset=&SelectVectorFromArray(
     Direct Array Index=&ModuloInt(A=&ReturnParticleIDAttributes(), B=4)
     Vector Selection Array=DataInterface() {propertyValues {FloatData [<10 0 0>, <-10 0 0>, <-7 5 0>, <7 -5 0>]}}
    )
    Skeletal Mesh=$System.Skeletal Mesh
    Use Sampled Mesh Attribute as Ribbon ID=false
   )
   /SetVariables(Particles.Position=&self:Select_Pos(
     bool:NiagaraBool=&SetBoolByIntComparison(A=&SpawnGroup(), B=1, Comparison Type=::NewEnumerator4)
     False Position:NiagaraPosition=$Particles.ShapeLocation.ShapePosition
     True Position:NiagaraPosition=$Particles.SkeletalMeshLocation.SampledPosition
    )) {isSetParameters=true}
  ]
  rendererClasses [&NiagaraSpriteRendererProperties]
  renderers [
   {rendererClass=&NiagaraSpriteRendererProperties}
   {rendererClass=&NiagaraSpriteRendererProperties, rendererIndex=1}
  ]
 }
 MainElectricity {
  emitterUpdateScript [
   &EmitterState(Life Cycle Mode=::NewEnumerator1, Loop Duration=1, UseLoopDelay=false)
   /SetVariables(Emitter.AudioSpawnRate:NiagaraFloat=&FloatFromCurve(
     FloatCurve=DataInterface() {propertyValues {Curve {keys [<at 0 1.5 RCIM_Cubic>, {arriveTangent=5.75, interpMode=RCIM_Cubic, leaveTangent=5.75, time=0.55, value=2.3068}, <at 1 7.25 RCIM_Cubic>]}}}
     CurveIndex=$System.Audio_DrumKick
    )) {isSetParameters=true}
   &SpawnRate(SpawnRate=$Emitter.AudioSpawnRate, Use Spawn Probability=false)
   &SpawnRate(SpawnGroup=1, SpawnRate=$Emitter.AudioSpawnRate, Use Spawn Probability=false) {module=SpawnRate001}
   &self:SpawnOnConditionMet(
    SpawnCondition:NiagaraBool=&SetBoolByFloatComparison(A=$System.Audio_DrumKick, B=0.99)
    Use Spawn Probability=false
    UseLoopCountLimit=false
   ) {enabled=false, module=SpawnOnConditionMet}
  ]
  particleSpawnScript [
   &InitializeParticle(
    Color=$System.MainColor
    Lifetime=&RandomRangeFloat(Minimum=0.35, Maximum=0.65)
    Mesh Renderer Array Visibility Mode=::NewEnumerator3
    Mesh Renderer Info=DataInterface() {propertyValues {MeshRenderer=&self:MainElectricity_0.NiagaraMeshRendererProperties_1}}
    Mesh Scale=&Multiply_Vector(
     A=&RandomRangeVector(Minimum=<0.125 0.125 0.125>, Maximum=<0.15 0.15 0.175>)
     B=&MakeVector(
      X=1
      Y=1
      Z=&Lerp_Float(
       A=0.7
       B=1.1
       Alpha=&FloatFromCurve(
        FloatCurve=DataInterface() {propertyValues {Curve {keys [
            {arriveTangent=4.1785, interpMode=RCIM_Cubic, time=0, value=0}
            {arriveTangent=1, interpMode=RCIM_Cubic, leaveTangent=1, time=0.5, value=0.2}
            {interpMode=RCIM_Cubic, leaveTangent=0.0224, time=1, value=1}
           ]}}}
        CurveIndex=$System.Audio_DrumKick
       )
      )
     )
    )
    Mesh Scale Mode=::NewEnumerator4
    Sprite Rotation Mode=::NewEnumerator2
    Sprite Size Mode=::NewEnumerator3
    UsePositionOffset=false
    Write Material Random=true
   )
   /SetVariables(Particles.SpawnGroup=$Emitter.SpawnGroup) {isSetParameters=true}
   /SetVariables(Particles.VisibilityTag=$Particles.SpawnGroup) {isSetParameters=true}
   &PartitionParticles(
    Number Of Partitions=4
    Particles Per Partition:NiagaraInt32=&MaskIntBySpawnGroup(Masked Int=2, Passthrough Int=1, Spawn Group Mask=0)
    Partition Mode:ENiagara_PartitionMode=::NewEnumerator1
   )
   &SkeletalMeshLocation(
    Bone / Socket Index=$Particles.PartitionParticles.Partition
    Bone Sampling Mode=::NewEnumerator4
    Interpolate Mesh Sampling=false
    MaskSpawnGroup=false
    Orientation Sampling=::NewEnumerator1
    Skeletal Mesh=$System.Skeletal Mesh
   ) {module=SkeletalMeshLocation001}
   /SetVariables(Particles.OrientationVec:Vector3f=&self:FindElectricityDirection(
     AudioIntensity:NiagaraFloat=$System.AudioIntensity
     DiveIntensity:NiagaraFloat=$System.Dive
     SourcePosition:NiagaraPosition=$Particles.Position
     TargetPosition:NiagaraPosition=$System.SkullPosition
    )) {isSetParameters=true}
   &InitialMeshOrientation(
    Mesh Axis To Orient=<0 0 1>
    Mesh Orientation Mode=::NewEnumerator2
    Orientation Coordinate Space=::World
    Orientation Vector=<0 0 1>
    Rotation=&MakeVector(X=&RandomRangeFloat(Minimum=0, Maximum=360), Y=0, Z=0)
    Use Rotation Vector=true
   )
   /SetVariables(Particles.FlareVizTag:NiagaraInt32=&RandomRangeInt(Minimum=0, Maximum=2)) {isSetParameters=true}
   /SetVariables(Particles.BraceletOrientation:Quat4f=/Niagara/DynamicInputs/Rotation/RandomQuaternion()) {isSetParameters=true}
  ]
  particleUpdateScript [
   &ParticleState()
   /SetVariables(
    Particles.LightningLength:NiagaraFloat=&FloatFromCurve(
     FloatCurve=DataInterface() {propertyValues {Curve {keys [<at 0 0.9>,<at 1 1>]}}}
     Scale Curve=&Lerp_Float(A=1, B=1.4, Alpha=$System.Dive)
    )
    Particles.LightningShapeExpand:NiagaraFloat=&FloatFromCurve(
     FloatCurve=DataInterface() {propertyValues {Curve {keys [
         {arriveTangent=4.1785, interpMode=RCIM_Cubic, leaveTangent=4.1785, tangentMode=RCTM_User, time=0, value=0}
         {arriveTangent=1.1179, interpMode=RCIM_Cubic, leaveTangent=1.1179, tangentMode=RCTM_User, time=0.2042, value=0.6228}
         {arriveTangent=0.0224, interpMode=RCIM_Cubic, leaveTangent=0.0224, tangentMode=RCTM_User, time=1, value=1}
        ]}}}
     Scale Curve=&Lerp_Float(A=0.5, B=0.35, Alpha=$System.Dive)
    )
   ) {isSetParameters=true, module=SetVariables_#26}
   &ScaleMeshSize(Scale Factor=&MakeVector(
     X=$Particles.LightningShapeExpand
     Y=$Particles.LightningShapeExpand
     Z=&Multiply_Float(A=$Particles.LightningLength, B=&Lerp_Float(A=0.8, B=1.25, Alpha=$System.Audio_DrumKick))
    ))
   &UpdateMeshOrientation(Facing Interpolation=false, Orientation Method=::NewEnumerator1, Override Quat=false, TransformFacing=false, TransformUp=false) {enabled=false}
   &SkeletalMeshLocation(
    Bone / Socket Index=$Particles.PartitionParticles.Partition
    Bone Sampling Mode=::NewEnumerator4
    MaskSpawnGroup=false
    Orientation Sampling=::NewEnumerator1
    Skeletal Mesh=$System.Skeletal Mesh
   )
   /SetVariables(Particles.OrientationVec:Vector3f=/Niagara/DynamicInputs/Lerp/SmoothLerpOverTime_Vector(
     Convergence Rate:NiagaraFloat=50
     Smooth Value:Vector3f=$Particles.OrientationVec
     Target Value:Vector3f=&Lerp_Vector(A=<0 0 -1>, B=<0 0 -1>, Alpha=$System.Dive)
    )) {enabled=false, isSetParameters=true}
   /SetVariables(Particles.OrientationVec:Vector3f=&self:FindElectricityDirection(
     AudioIntensity:NiagaraFloat=$System.AudioIntensity
     DiveIntensity:NiagaraFloat=$System.Dive
     SourcePosition:NiagaraPosition=$Particles.Position
     TargetPosition:NiagaraPosition=$System.SkullPosition
    )) {enabled=false, isSetParameters=true}
   &UpdateMeshOrientation(
    Facing Coordinate Space:ENiagaraCoordinateSpace=::World
    Facing Interpolation=false
    Orientation Method=::NewEnumerator1
    Override Quat=false
    TransformFacing=true
    TransformUp=true
    Up Direction:Vector3f=<0 0 0>
   ) {enabled=false, module=UpdateMeshOrientation001}
   /SetVariables(Particles.MeshOrientation=&OrientMeshToVector(Look At Direction=$System.SysLightningFaceDir, Reference Axis Vector=<1 0 0>)) {isSetParameters=true, module=SetVariables_#27}
   &DynamicMaterialParameters(
    Index 0 Param 1=&FloatFromCurve(
     FloatCurve=DataInterface() {propertyValues {Curve {keys [<at 0 1>,<at 1 0.5>]}}}
     Scale Curve=&Lerp_Float(A=0.3, B=1.25, Alpha=$System.Audio_DrumKick)
    )
    Index 0 Param 3=&RandomRangeFloat(Minimum=0, Maximum=22)
    Index 0 Param 4=0.5
    Index 1 Param 1=&FloatFromCurve(
     FloatCurve=DataInterface() {propertyValues {Curve {keys [
         <at 0 0>
         {arriveTangent=1, interpMode=RCIM_Cubic, leaveTangent=1, time=0.3, value=0}
         {interpMode=RCIM_Cubic, leaveTangent=-8.0089, time=1, value=1}
        ]}}}
     Scale Curve=0.1
    )
    Index 1 Param 2=1
    Index 1 Param 3=&FloatFromCurve(FloatCurve=DataInterface() {propertyValues {Curve {keys [<at 0.5 1 RCIM_Cubic>,<at 1 0 RCIM_Cubic>]}}}, Scale Curve=2)
    Index 2 Param 1=&FloatFromCurve(
     FloatCurve=DataInterface() {propertyValues {Curve {keys [{tangentMode=RCTM_User, time=0, value=0}, {tangentMode=RCTM_User, time=1, value=1}]}}}
     Scale Curve=0
    )
    Index 2 Param 2=2
    Index 3 Param 1=&MakeFloatFromVector(Vector=Linked() {linkedVariable {name=System.SkullPosition, type=NiagaraPosition}}, Channel=::NewEnumerator0)
    Index 3 Param 2=&MakeFloatFromVector(Vector=Linked() {linkedVariable {name=System.SkullPosition, type=NiagaraPosition}}, Channel=::NewEnumerator1)
    Index 3 Param 3=&MakeFloatFromVector(Vector=Linked() {linkedVariable {name=System.SkullPosition, type=NiagaraPosition}}, Channel=::NewEnumerator2)
    Index 3 Param 4=30
    Module.Index0Param0Write=true
    Module.Index0Param1Write=true
    Module.Index0Param2Write=true
    Module.Index0Param3Write=true
    Module.Index1Param0Write=true
    Module.Index1Param1Write=true
    Module.Index1Param2Write=true
    Module.Index1Param3Write=true
    Module.Index2Param0Write=true
    Module.Index2Param1Write=true
    Module.Index2Param2Write=true
    Module.Index2Param3Write=true
    Module.Index3Param0Write=true
    Module.Index3Param1Write=true
    Module.Index3Param2Write=true
    Module.Index3Param3Write=true
    Write Parameter Index 1=true
    Write Parameter Index 2=true
    Write Parameter Index 3=true
   )
   &UpdateMeshOrientation(Rotation Rate=2) {module=UpdateMeshOrientation002}
   &ScaleColor(
    Scale Alpha=&FloatFromCurve(FloatCurve=DataInterface() {propertyValues {Curve {keys [
         {tangentMode=RCTM_User, time=0, value=1}
         {arriveTangent=-0.0405, interpMode=RCIM_Cubic, leaveTangent=-0.0405, tangentMode=RCTM_User, time=0.7009, value=1}
         {arriveTangent=-8.0089, interpMode=RCIM_Cubic, leaveTangent=-8.0089, tangentMode=RCTM_User, time=1, value=0}
        ]}}})
    Scale RGB=&VectorFromFloat(Value=&FloatFromCurve(
      FloatCurve=DataInterface() {propertyValues {Curve {keys [
          {arriveTangent=9.0725, interpMode=RCIM_Cubic, time=0, value=0}
          <at 0.1408 2 RCIM_Cubic>
          {interpMode=RCIM_Cubic, leaveTangent=-0.0405, time=1, value=0}
         ]}}}
      Scale Curve=&Multiply_Float(A=&FN_FrontEndFloatModifier(Gameplay=4, Front End=8), B=$System.AudioBrightnessMultiplier)
     ))
    ScaleA=true
    ScaleRGB=true
   ) {module=ScaleColor001}
   &ScaleColor(
    Scale RGB=&Multiply_VectorByFloat(
     Vector=&VectorFromFloat(Value=&Lerp_Float(
       A=0.5
       B=2
       Alpha=&Waveform(
        Waveform Count=::NewEnumerator1
        Global Phase=$Particles.Age
        "Waveform [1]"=::NewEnumerator12
        "[1] Frequency"=16
        "[1] Amplitude Min/Max"=<0 1>
        "[2] Average"=false
        "Waveform [2]"=::NewEnumerator12
        "[2] Input Position"=0
        "[2] Phase"=$Emitter.Age
        "[2] Phase Scale"=1
        "[2] Compound Frequency"=2
        "[2] Frequency"=30
        "[2] Amplitude Min/Max"=<0 1>
        "[2] Scale"=3
        "[2] Bias"=0
        "[2] Exponent"=1
        "[2] Pulse Duration"=0.1
        "[2] Clip Triangle Min/Max"=<0 0>
        "[2] Random Seed"=0
       )
      ))
     Float=&FN_FrontEndFloatModifier(Gameplay=1, Front End=0.35)
    )
    ScaleA=true
    ScaleRGB=true
   ) {enabled=false}
   &InheritSourceMovement(ApplyForce=false, ApplyPosition=true, ApplyPositionFalloff=false) {enabled=false}
   &CameraOffset(Camera Offset Amount=10)
   /SetVariables(
    Particles.FlareColor=&MultiplyLinearColor(
     Linear Color=$System.MainColor
     Scale Factor=&MakeFloat4FromVectorAndFloat(
      "Vector (RGB)"=&VectorFromFloat(Value=&FloatFromCurve(
        FloatCurve=DataInterface() {propertyValues {Curve {keys [
            {arriveTangent=9.0725, interpMode=RCIM_Cubic, leaveTangent=9.0725, tangentMode=RCTM_User, time=0, value=0}
            {arriveTangent=-2.5249, interpMode=RCIM_Cubic, leaveTangent=-2.5249, tangentMode=RCTM_User, time=0.1408, value=0.9847}
            {arriveTangent=-1.146, interpMode=RCIM_Cubic, leaveTangent=-1.146, time=0.3, value=0.2}
            {arriveTangent=-0.0405, interpMode=RCIM_Cubic, leaveTangent=-0.0405, tangentMode=RCTM_User, time=1, value=0}
           ]}}}
        Scale Curve=3
       ))
      "Float (Alpha)"=1
     )
    )
    Particles.FlarePosition:NiagaraPosition=&AddVectorToPosition(Position=$Particles.Position, Vector=<0 0 5>)
    Particles.FlareSize:Vector2f=&Multiply_Vector2D_ByFloat(
     Float=&FloatFromCurve(FloatCurve=DataInterface() {propertyValues {Curve {keys [
          {arriveTangent=9.0725, interpMode=RCIM_Cubic, leaveTangent=9.0725, tangentMode=RCTM_User, time=0, value=0}
          {arriveTangent=-2.5249, interpMode=RCIM_Cubic, leaveTangent=-2.5249, tangentMode=RCTM_User, time=0.1408, value=0.9847}
          {arriveTangent=-1.146, interpMode=RCIM_Cubic, leaveTangent=-1.146, time=0.3, value=0.25}
          {arriveTangent=-0.0405, interpMode=RCIM_Cubic, leaveTangent=-0.0405, tangentMode=RCTM_User, time=1, value=0}
         ]}}})
     Vector2D=<50 50>
    )
   ) {isSetParameters=true}
   /SetVariables(Particles.CoreColor:LinearColor=&MultiplyLinearColor(Linear Color=$Particles.FlareColor, Scale Factor=<2 2 2 1>)) {isSetParameters=true}
   /SetVariables(Particles.BraceletSize:Vector3f=&VectorFromFloat(Value=&FloatFromCurve(
      FloatCurve=DataInterface() {propertyValues {Curve {keys [
          {arriveTangent=4.1785, interpMode=RCIM_Cubic, leaveTangent=4.1785, tangentMode=RCTM_User, time=0, value=0}
          {arriveTangent=1.1179, interpMode=RCIM_Cubic, leaveTangent=1.1179, tangentMode=RCTM_User, time=0.2042, value=0.6228}
          {arriveTangent=0.0224, interpMode=RCIM_Cubic, leaveTangent=0.0224, tangentMode=RCTM_User, time=1, value=1}
         ]}}}
      Scale Curve=0.95
     ))) {isSetParameters=true}
  ]
  rendererClasses [&NiagaraMeshRendererProperties]
  renderers [{rendererClass=&NiagaraMeshRendererProperties, rendererIndex=0}]
 }
 MainSkull {
  emitterUpdateScript [
   &EmitterState()
   &SpawnBurst_Instantaneous(Use Spawn Probability=false, UseLoopCountLimit=false)
   &SpawnBurst_Instantaneous(Spawn Group=1, Use Spawn Probability=false, UseLoopCountLimit=false) {module=SpawnBurst_Instantaneous001}
   &SpawnBurst_Instantaneous(Spawn Group=2, Use Spawn Probability=false, UseLoopCountLimit=false) {module=SpawnBurst_Instantaneous002}
   /SetVariables(Emitter.SkullBrightness:NiagaraFloat=&FloatFromCurve(
     FloatCurve=DataInterface() {propertyValues {Curve {keys [{arriveTangent=1, interpMode=RCIM_Cubic, time=0.25, value=0.5}, <at 1 1 RCIM_Cubic>]}}}
     CurveIndex=$System.Audio_DrumKick
    )) {isSetParameters=true}
   /SetVariables(Emitter.EyeScale:NiagaraFloat=&Lerp_Float(
     A=0.8
     B=1.5
     Alpha=&RemapRange(Input Value To Remap=$NPC.NPC_Audio_Reactive_Parameters.Amplitude_Music, Input Min=0.8, Input Max=1, Output Min=0, Output Max=1)
    )) {isSetParameters=true}
  ]
  particleSpawnScript [
   &InitializeParticle(
    Color=$System.MainColor
    Lifetime=$Emitter.CurrentLoopDuration
    Mesh Scale Mode=::NewEnumerator3
    Mesh Uniform Scale=&Multiply_Float(A=0.14, B=&FN_FrontEndFloatModifier(Gameplay=1, Front End=1))
    Sprite Rotation Mode=::NewEnumerator2
    Sprite Size Mode=::NewEnumerator3
    Uniform Sprite Size=34.3195
    UsePositionOffset=false
    Write Material Random=true
   )
   /SetVariables(Particles.Position=$System.SkullPosition) {isSetParameters=true}
   /SetVariables(Particles.SpawnGroup=$Emitter.SpawnGroup) {isSetParameters=true}
   /SetVariables(Particles.VisibilityTag=&MaskIntBySpawnGroup(Masked Int=0, Passthrough Int=1, Spawn Group Mask=0)) {isSetParameters=true}
   /SetVariables(Particles.SkullScaleMultiplier=1) {isSetParameters=true}
  ]
  particleUpdateScript [
   &ParticleState()
   /SetVariables(Particles.SkullScaleMultiplier:NiagaraFloat=/Niagara/DynamicInputs/Lerp/SmoothLerpOverTime_Float(
     Convergence Rate:NiagaraFloat=8
     Smooth Value:NiagaraFloat=$Particles.SkullScaleMultiplier
     Target Value:NiagaraFloat=&RemapRange(
      Input Value To Remap=&MakeFloatFromLinearColor(LinearColor=$NPC.NPC_Audio_Reactive_Parameters.Spectrum_Music, Channel:ENiagaraLinearColor_Channels=::NewEnumerator3)
      Input Min=0.45
      Input Max=0.65
      Output Min=0.9
      Output Max=1.2
     )
    )) {isSetParameters=true}
   &ScaleMeshSize(Scale Factor=&VectorFromFloat(Value=$Particles.SkullScaleMultiplier))
   &ScaleSpriteSize(
    Scale Sprite Size Mode=::NewEnumerator0
    Uniform Scale Factor=&Lerp_Float(A=0.75, B=1.5, Alpha=$System.AudioIntensity)
   )
   &SpriteRotationRate(Rotation Rate=&RandomRangeFloat(Minimum=-10, Maximum=10))
   &UpdateMeshOrientation() {enabled=false}
   &SolveForcesAndVelocity(Clamp Velocity=false, Limit Acceleration=false)
   &ScaleColor(
    Scale RGB=&Multiply_VectorByFloat(
     Vector=&VectorFromFloat(Value=&Lerp_Float(A=1, B=4, Alpha=$Emitter.SkullBrightness))
     Float=&FN_FrontEndFloatModifier(Gameplay=1, Front End=0.5)
    )
    ScaleA=true
    ScaleRGB=true
   )
   /SetVariables(Particles.MeshOrientation=&MultiplyQuaternion(
     Quaternion A=&MultiplyQuaternion(
      Quaternion A=$System.SkullRotation
      Quaternion B=&MakeQuatFromAxisAngle(Axis=<0 1 0>, AngleInDegrees=90)
     )
     Quaternion B=&MakeQuatFromAxisAngle(
      Axis=<1 0 0>
      AngleInDegrees=&FN_FrontEndFloatModifier(
       Gameplay=&Lerp_Float(
        A=&FloatFromCurve(FloatCurve=DataInterface() {propertyValues {Curve {keys [<at 0.35 45>,<at 0.75 0>]}}}, CurveIndex=$System.SkullFacingBlend)
        B=90
        Alpha=$System.Dive
       )
       Front End=65
      )
     )
    )) {isSetParameters=true}
   /SetVariables(Particles.MeshOrientation=&LerpQuaternion(
     Quaternion A:Quat4f=&MultiplyQuaternion(
      Quaternion A=$Particles.MeshOrientation
      Quaternion B=&MakeQuatFromAxisAngle(Axis=<0 0 1>, AngleInDegrees=&FN_FrontEndFloatModifier(Gameplay=180, Front End=0))
     )
     Quaternion B:Quat4f=$Particles.MeshOrientation
     Lerp Factor:NiagaraFloat=&MakeCustomFloatFromBool(
      Bool=true
      True Float=0
      False Float=&FN_FrontEndFloatModifier(Gameplay=$System.Dive, Front End=1)
     )
     Enable Spherical Linear Interpolation:NiagaraBool=true
    )) {isSetParameters=true}
   /SetVariables(Particles.MeshOrientation=&MultiplyQuaternion(
     Quaternion A=$Particles.MeshOrientation
     Quaternion B=&MakeQuatFromAxisAngle(
      Axis=<0 0 1>
      AngleInDegrees=&Lerp_Float(
       A=&FloatFromCurve(
        FloatCurve=DataInterface() {propertyValues {Curve {keys [<at -1 -70>,<at -0.5 0>,<at 0.5 0>,<at 1 35>]}}}
        CurveIndex=&RemapRange(Input Value To Remap=$System.SideVelocity, Input Min=-500, Input Max=500, Output Min=-1, Output Max=1)
       )
       B=0
       Alpha=$System.Dive
      )
     )
    )) {isSetParameters=true}
   /SetVariables(Particles.MeshOrientation=&MultiplyQuaternion(
     Quaternion A=$Particles.MeshOrientation
     Quaternion B=&MakeQuatFromAxisAngle(Axis=<1 0 0>, AngleInDegrees=&Lerp_Float(A=0, B=-10, Alpha=$System.Audio_DrumKick))
    )) {isSetParameters=true}
   &UpdateMeshOrientation(
    Facing Direction=&ReturnCameraVector(Camera Vectors=::NewEnumerator1, Coordinate Space=::World)
    Facing Interpolation=false
    Orientation Method=::NewEnumerator1
    Override Quat=false
    Side Coordinate Space:ENiagaraCoordinateSpace=::World
    Side Direction:Vector3f=<1 0 1>
    Stabilization Mode:ENiagara_OrientVectorCount=::NewEnumerator3
    TransformFacing=false
    TransformSide=true
    TransformUp=true
    Up Coordinate Space:ENiagaraCoordinateSpace=::World
    Up Direction:Vector3f=<1 0 0>
   ) {enabled=false, module=UpdateMeshOrientation002}
   /SetVariables(Particles.Position=&self:ChooseSkullElemPos(
     Index:NiagaraInt32=$Particles.SpawnGroup
     LeftEyePos:NiagaraPosition=/Niagara/DynamicInputs/Math/TransformVectorByMatrix(
      Vector:Vector3f=<7 16 -2>
      Transformation Matrix:NiagaraMatrix=/Niagara/DynamicInputs/Math/MakeTransform(
       Inverse:NiagaraBool=false
       Position:Vector3f=Linked() {linkedVariable {name=System.SkullPosition, type=NiagaraPosition}}
       Local Offset:Vector3f=<0 0 0>
       Scale:Vector3f=&Multiply_VectorByFloat(Vector=<1 1 1>, Float=$Particles.SkullScaleMultiplier)
       Rotation Interface:ENiagara_RotationDefinitionApproach=::NewEnumerator1
       Axis:Vector3f=?
       Quaternion:Quat4f=$StackContext.MeshOrientation
       Rotation Matrix:NiagaraMatrix=?
       X Vector:Vector3f=?
       Yaw:NiagaraFloat=?
       Angle:NiagaraFloat=?
       Pitch:NiagaraFloat=?
       Y Vector:Vector3f=?
       Roll:NiagaraFloat=?
       Z Vector:Vector3f=?
      )
      Transform Vector as Position:NiagaraBool=true
     )
     RightEyePos:NiagaraPosition=&TransformVectorByMatrix(
      Vector:Vector3f=<-7 16 -2>
      Transformation Matrix:NiagaraMatrix=&MakeTransform(
       Inverse:NiagaraBool=false
       Position:Vector3f=Linked() {linkedVariable {name=System.SkullPosition, type=NiagaraPosition}}
       Local Offset:Vector3f=<0 0 0>
       Scale:Vector3f=&Multiply_VectorByFloat(Vector=<1 1 1>, Float=$Particles.SkullScaleMultiplier)
       Rotation Interface:ENiagara_RotationDefinitionApproach=::NewEnumerator1
       Axis:Vector3f=?
       Quaternion:Quat4f=$StackContext.MeshOrientation
       Rotation Matrix:NiagaraMatrix=?
       X Vector:Vector3f=?
       Yaw:NiagaraFloat=?
       Angle:NiagaraFloat=?
       Pitch:NiagaraFloat=?
       Y Vector:Vector3f=?
       Roll:NiagaraFloat=?
       Z Vector:Vector3f=?
      )
      Transform Vector as Position:NiagaraBool=true
     )
     SkullPosition:NiagaraPosition=$System.SkullPosition
    )) {isSetParameters=true}
   &UpdateMeshOrientation(
    Facing Direction=&ReturnCameraVector(Camera Vectors=::NewEnumerator4, Coordinate Space=::World, Origin Position=$Particles.Position)
    Facing Interpolation=false
    Orientation Method=::NewEnumerator1
    Override Quat=false
    TransformFacing=false
    TransformUp=true
   ) {enabled=false, module=UpdateMeshOrientation001}
   /SetVariables(Particles.EyesColor:LinearColor=&MultiplyLinearColor(
     Linear Color=$System.MainColor
     Scale Factor=&MakeFloat4FromVectorAndFloat(
      "Vector (RGB)"=&VectorFromFloat(Value=&Lerp_Float(A=2, B=3.5, Alpha=$System.AudioBrightnessMultiplier))
      "Float (Alpha)"=&Multiply_Float(
       A=&FloatFromCurve(FloatCurve=DataInterface() {propertyValues {Curve {keys [<at 0 0.1>,<at 1 1>]}}}, CurveIndex=$Emitter.SkullBrightness)
       B=&FN_FrontEndFloatModifier(Gameplay=3, Front End=0.5)
      )
     )
    )) {isSetParameters=true}
   /SetVariables(
    Particles.SkullGlowColor:LinearColor=&MultiplyLinearColor(
     Linear Color=$Particles.Color
     Scale Factor=&MakeFloat4FromVectorAndFloat("Vector (RGB)"=<1 1 1>, "Float (Alpha)"=0.025)
    )
    Particles.SkullGlowSize:Vector2f=&Multiply_Vector2D_ByFloat(Float=$Emitter.EyeScale, Vector2D=<200 200>)
   ) {isSetParameters=true}
   /SetVariables(Particles.BlankDynParams:Vector4f=<1 0 0 0>) {isSetParameters=true}
   &DynamicMaterialParameters(
    Index 0 Param 1=&Lerp_Float(A=0, B=1, Alpha=$System.Audio_DrumKick)
    Index 0 Param 2=&Add_Float(
     A=&Waveform("Waveform [1]"=::NewEnumerator0, "[1] Frequency"=1.25, "[1] Amplitude Min/Max"=<-1 0.02>, "[1] Scale"=0.05)
     B=&Lerp_Float(A=0.08, B=-0.062, Alpha=$System.AudioIntensity)
    )
    Index 0 Param 3=&Waveform("Waveform [1]"=::NewEnumerator12, "[1] Frequency"=16, "[1] Amplitude Min/Max"=<0.48 0.52>)
    Index 0 Param 4=&FloatFromCurve(
     FloatCurve=DataInterface() {propertyValues {Curve {keys [<at -1 0.75 RCIM_Constant>,<at 0.5 0.125 RCIM_Constant>]}}}
     CurveIndex=&DotProduct(
      Vector A=&RotateVectorByQuaternion(Vector To Rotate=<0 -1 0>, Quaternion=$Particles.MeshOrientation)
      Vector B=&ReturnCameraVector(Camera Vectors=::NewEnumerator1, Coordinate Space=::World)
     )
    )
    Module.Index0Param0Write=true
    Module.Index0Param1Write=true
    Module.Index0Param2Write=true
    Module.Index0Param3Write=true
   )
   &FN_Light_Attributes(
    Color=$Particles.Color
    Create Color Attribute=true
    Create Exponent Attribute=true
    Create Radius Attribute=true
    Exponent=4
    Radius=100
    Scale Color=35
   )
   /SetVariables(
    Particles.EyeCoreColor:LinearColor=&MultiplyLinearColor(
     Linear Color=$Particles.EyesColor
     Scale Factor=&MakeFloat4FromVectorAndFloat("Vector (RGB)"=&VectorFromFloat(Value=5), "Float (Alpha)"=3)
    )
    Particles.EyeCoreSize:Vector2f=&Multiply_Vector2D(
     A=$Particles.SpriteSize
     B=&Multiply_Vector2D_ByFloat(
      Float=&Lerp_Float(
       A=0
       B=1
       Alpha=&RemapRange(Input Value To Remap=$NPC.NPC_Audio_Reactive_Parameters.Amplitude_Music, Input Min=0.8, Input Max=1, Output Min=0, Output Max=1)
      )
      Vector2D=<0.2 0.2>
     )
    )
    Particles.EyesCoreRotation:NiagaraFloat=&RandomRangeFloat(Minimum=0, Maximum=1)
   ) {isSetParameters=true}
  ]
  rendererClasses [
   &NiagaraMeshRendererProperties
   &NiagaraSpriteRendererProperties
   &NiagaraLightRendererProperties
  ]
  renderers [
   {rendererClass=&NiagaraMeshRendererProperties, rendererIndex=0}
   {rendererClass=&NiagaraSpriteRendererProperties, rendererIndex=1}
   {rendererClass=&NiagaraSpriteRendererProperties, rendererIndex=2}
   {rendererClass=&NiagaraLightRendererProperties, rendererIndex=3}
   {rendererClass=&NiagaraSpriteRendererProperties, rendererIndex=4}
  ]
 }
 PointHelpers {
  emitterUpdateScript [
   &EmitterState(Life Cycle Mode=::NewEnumerator1, Loop Behavior=::NewEnumerator1, Loop Duration Mode=::NewEnumerator1, UseLoopDelay=false)
   &SpawnBurst_Instantaneous(Spawn Count=4, Use Spawn Probability=false, UseLoopCountLimit=false)
   /Game/Effects/Niagara/Modules/Spawn/SpawnRateFromMultipleSpawnGroups_2(Global Spawnrate=8) {enabled=false}
  ]
  particleSpawnScript [
   /Niagara/Modules/Spawn/Initialization/InitializeParticle(
    Color=<rgba 0.9062 0.4123 0.1067 0.6>
    Lifetime=0.6
    Module.Write Color=true
    Module.Write Lifetime=true
    Module.Write Mass=true
    Module.Write Material Random=true
    Module.Write Position=false
    Module.Write Scale=false
    Module.Write SpriteRotation=false
    Module.Write SpriteSize=true
    Position=&SimulationPosition()
    Sprite Size=<40 40>
   ) {enabled=false}
   /Niagara/Modules/Spawn/Initialization/V2/InitializeParticle(
    Color=&MultiplyLinearColor(
     Linear Color=$System.MainColor
     Scale Factor=&MakeFloat4FromVectorAndFloat("Vector (RGB)"=&VectorFromFloat(Value=3), "Float (Alpha)"=1)
    )
    Lifetime=0.6
    Mass Mode=::NewEnumerator1
    Sprite Size=<50 50>
    Sprite Size Mode=::NewEnumerator4
    UsePositionOffset=false
    Write Material Random=true
   ) {module=InitializeParticle001}
   &CameraOffset()
   &PartitionParticles(Number Of Partitions=4)
  ]
  particleUpdateScript [
   &ParticleState(
    Kill Particles When Lifetime Has Elapsed=false
    Let Infinitely Lived Particles Die When Emitter Deactivates=true
    Loop Particles Lifetime=true
   )
   &SkeletalMeshLocation(
    Bone / Socket Index=$Particles.PartitionParticles.Partition
    Bone Sampling Mode=::NewEnumerator4
    Inherit Velocity=false
    Interpolate Mesh Sampling=false
    MaskSpawnGroup=false
    Sampled Position Offset=&SelectVectorFromArray(Direct Array Index=$Particles.PartitionParticles.Partition, Vector Selection Array=DataInterface() {propertyValues {FloatData []}})
    Sampled Position Offset Coordinate Space:ENiagaraMeshTransforms=::NewEnumerator3
    Skeletal Mesh=$System.Skeletal Mesh
   )
   &ScaleColor(
    Scale Alpha=&FloatFromCurve(
     FloatCurve=DataInterface() {propertyValues {Curve {keys [
         <at 0.0001 0 RCIM_Cubic>
         {arriveTangent=1.6347, interpMode=RCIM_Cubic, leaveTangent=1.6347, time=0.1393, value=0.6251}
         {arriveTangent=0.8334, interpMode=RCIM_Cubic, leaveTangent=0.8334, time=0.3202, value=0.5233}
         {arriveTangent=-0.4566, interpMode=RCIM_Cubic, leaveTangent=-0.4566, time=0.4613, value=0.8934}
         {arriveTangent=-0.922, interpMode=RCIM_Cubic, leaveTangent=-0.922, time=0.6241, value=0.3845}
         {arriveTangent=-1.0215, interpMode=RCIM_Cubic, leaveTangent=-1.0215, time=0.7724, value=0.6066}
         <at 1.0006 0 RCIM_Cubic>
        ]}}}
     Scale Curve=0.25
    )
    ScaleA=true
    ScaleRGB=false
   )
   /SetVariables(
    Particles.FlareColor=&MultiplyLinearColor(
     Linear Color=$Particles.Color
     Scale Factor=&MakeFloat4FromVectorAndFloat("Vector (RGB)"=&VectorFromFloat(Value=0.1), "Float (Alpha)"=1)
    )
    Particles.FlareSize:Vector2f=&Multiply_Vector2D_ByFloat(Float=0.5, Vector2D=$Particles.SpriteSize)
    Particles.float=15
   ) {isSetParameters=true}
   /SetVariables(Particles.VisibilityTag=&MakeCustomIntFromBool(Bool=&SetBoolByFloatComparison(A=$System.CloudDot, B=0.5), True INT=0, False INT=1)) {isSetParameters=true}
   /SetVariables() {enabled=false, isSetParameters=true}
   /SetVariables(Particles.SharpFlareColor:LinearColor=<rgba 555 1 1 1>) {isSetParameters=true}
   /SetVariables(Particles.SharpFlareSize:Vector2f=<500 500>) {isSetParameters=true}
  ]
  rendererClasses [&NiagaraSpriteRendererProperties]
  renderers [
   {rendererClass=&NiagaraSpriteRendererProperties}
   {rendererClass=&NiagaraSpriteRendererProperties, rendererIndex=1}
   {rendererClass=&NiagaraSpriteRendererProperties, rendererIndex=2}
   {rendererClass=&NiagaraSpriteRendererProperties, rendererIndex=3}
  ]
 }
 Sparks {
  simTarget=GPUComputeSim
  emitterUpdateScript [
   &EmitterState()
   &SpawnRate(
    SpawnRate=&Multiply_Float(
     A=&Multiply_Float(A=100, B=&Waveform("Waveform [1]"=::NewEnumerator12, "[1] Frequency"=12, "[1] Amplitude Min/Max"=<0.5 1.5>))
     B=$System.Audio_DrumKick
    )
    Use Spawn Probability=false
   )
  ]
  particleSpawnScript [
   &InitializeParticle(
    Color=&RandomRangeLinearColor(Minimum=<rgba 0.0867 0.876 2 1>, Maximum=<rgba 0.1314 1.0657 2 1>)
    Lifetime Max=0.4
    Lifetime Min=0.2
    Lifetime Mode=::NewEnumerator1
    Sprite Size Max=<1 4>
    Sprite Size Min=<0.5 2>
    Sprite Size Mode=::NewEnumerator2
    UsePositionOffset=false
    Write Material Random=true
   )
   &InheritVelocity(
    Inherited Velocity Amount Scale=<-10 -10 -10>
    Inherited Velocity Speed Limit=488.008
    Speed Limit=true
    UseSpeedThreshold=false
    Velocity Source=&TransformVector(Vector=$Engine.Owner.Velocity, Source Space=::World, Destination Space=::Simulation, Apply Scale=true)
   ) {enabled=false}
   &PartitionParticles(Number Of Partitions=4)
   &SkeletalMeshLocation(
    Bone / Socket Index=$Particles.PartitionParticles.Partition
    Bone Sampling Mode=::NewEnumerator4
    Interpolate Mesh Sampling=false
    MaskSpawnGroup=false
    Orientation Sampling=::NewEnumerator1
    Skeletal Mesh=$System.Skeletal Mesh
   ) {module=SkeletalMeshLocation001}
   &AddVelocity(
    Cone Axis=$System.SysLightningFaceDir
    Module.UseConeFalloff=false
    Module.UseVelDistribution=false
    Override Seed=false
    Rotation Coordinate Space=::World
    Velocity Mode=::NewEnumerator2
    Velocity Speed=&Multiply_Float(
     A=&RandomRangeFloat(Minimum=250, Maximum=750)
     B=&Waveform("Waveform [1]"=::NewEnumerator12, "[1] Frequency"=12, "[1] Amplitude Min/Max"=<0.5 1.5>)
    )
   )
  ]
  particleUpdateScript [
   &ParticleState()
   &Drag(Drag=12, Use Linear Drag=true, Use Rotational Drag=true)
   &ScaleColor(
    Scale Alpha=&FloatFromCurve(FloatCurve=DataInterface() {propertyValues {Curve {keys [
         {arriveTangent=9.0725, interpMode=RCIM_Cubic, leaveTangent=9.0725, tangentMode=RCTM_User, time=0, value=0}
         {arriveTangent=-2.5249, interpMode=RCIM_Cubic, leaveTangent=-2.5249, tangentMode=RCTM_User, time=0.1408, value=0.9847}
         {arriveTangent=-0.0405, interpMode=RCIM_Cubic, leaveTangent=-0.0405, tangentMode=RCTM_User, time=1, value=0}
        ]}}})
    Scale RGB=&VectorFromFloat(Value=&FloatFromCurve(FloatCurve=DataInterface() {propertyValues {Curve {keys [<at 0 1>,<at 0.1 0.1>,<at 1 0>]}}}, Scale Curve=100))
    ScaleA=true
    ScaleRGB=true
   )
   &CurlNoiseForce(
    Noise Frequency=250
    Noise Strength=&FloatFromCurve(
     FloatCurve=DataInterface() {propertyValues {Curve {keys [
         {arriveTangent=4.1785, interpMode=RCIM_Cubic, leaveTangent=4.1785, tangentMode=RCTM_User, time=0, value=0}
         {arriveTangent=1.1179, interpMode=RCIM_Cubic, leaveTangent=1.1179, tangentMode=RCTM_User, time=0.2042, value=0.6228}
         {arriveTangent=0.0224, interpMode=RCIM_Cubic, leaveTangent=0.0224, tangentMode=RCTM_User, time=1, value=1}
        ]}}}
     Scale Curve=2500
    )
    Pan Noise=false
   )
   &SolveForcesAndVelocity(Clamp Velocity=false, Limit Acceleration=false)
   &ScaleSpriteSizeBySpeed(Max Scale Factor=<1.5 2.2347>, Min Scale Factor=<0.5 1>)
   &ScaleSpriteSize(Scale Factor=&Vector2DFromFloat(Value=&FloatFromCurve(
      FloatCurve=DataInterface() {propertyValues {Curve {keys [
          <at 0 0>
          <at 0.08 1.0301>
          {arriveTangent=-6.3909, interpMode=RCIM_Cubic, leaveTangent=-6.3909, tangentMode=RCTM_User, time=0.6, value=1.05}
          <at 1 0>
         ]}}}
      Scale Curve=&Multiply_Float(A=&Lerp_Float(A=1, B=2, Alpha=$System.DistanceScalar), B=$User.FXBoundScalar)
     )))
  ]
  rendererClasses [&NiagaraSpriteRendererProperties]
  renderers [{rendererClass=&NiagaraSpriteRendererProperties}]
 }
 SplitLightning {
  bEnabled=false
  emitterUpdateScript [
   &EmitterState()
   &SpawnRate(SpawnRate=20, Use Spawn Probability=false)
   &SpawnBurst_Instantaneous(Spawn Count=5, Spawn Time=0.005, Use Spawn Probability=false, UseLoopCountLimit=false) {enabled=false}
  ]
  particleSpawnScript [
   &InitializeParticle(
    Color=&RandomRangeLinearColor(Minimum=<rgba 1.2 1.2 1.2 0.5>, Maximum=<rgba 1.6 1.6 1.6 0.6>)
    Lifetime=&RandomRangeFloat(Minimum=0.35, Maximum=0.5)
    Mesh Scale Mode=::NewEnumerator3
    Mesh Uniform Scale=&RandomRangeFloat(Minimum=0.35, Maximum=0.6)
    UsePositionOffset=false
    Write Material Random=true
   )
   &AddVelocity(
    Cone Axis=<0 -1 0>
    Module.UseConeFalloff=false
    Module.UseVelDistribution=false
    Override Seed=false
    Velocity Mode=::NewEnumerator2
    Velocity Speed=0.1
   )
   &CameraOffset(Camera Offset Amount=5)
   &PartitionParticles(Number Of Partitions=4)
  ]
  particleUpdateScript [
   &ParticleState()
   &ScaleMeshSize(Scale Factor=&RandomRangeVector(Minimum=<1 1 1>, Maximum=<1.5 1.5 1.5>)) {module=ScaleMeshSize001}
   &SkeletalMeshLocation(
    Bone / Socket Index=$Particles.PartitionParticles.Partition
    Bone Sampling Mode=::NewEnumerator4
    MaskSpawnGroup=false
    Orientation Sampling=::NewEnumerator1
    Skeletal Mesh=$System.Skeletal Mesh
   ) {module=SkeletalMeshLocation001}
   &ScaleMeshSize(Scale Factor=&MakeVector(
     X=1
     Y=1
     Z=&FloatFromCurve(FloatCurve=DataInterface() {propertyValues {Curve {keys [
          {interpMode=RCIM_Cubic, leaveTangent=0.0966, tangentMode=RCTM_Break, time=0, value=1}
          {arriveTangent=0.0951, interpMode=RCIM_Cubic, leaveTangent=-0.1745, tangentMode=RCTM_Break, time=1, value=1.07}
         ]}}})
    ))
   &DynamicMaterialParameters(
    Index 0 Param 4=&MakeCustomFloatFromBool(
     Bool=false
     True Float=10
     False Float=&SelectFloatFromArray(
      Array Sampling Mode=::NewEnumerator0
      Direct Array Index=?
      Float Selection Array=DataInterface() {propertyValues {FloatData [3,9,9]}}
      Random Seed=0
     )
    )
    Index 1 Param 1=1
    Index 1 Param 2=2
    Index 1 Param 3=&FloatFromCurve(FloatCurve=DataInterface() {propertyValues {Curve {keys [{interpMode=RCIM_Cubic, tangentMode=RCTM_Break, time=0.5, value=-0.5}, <at 1 0.1>]}}})
    Index 2 Param 1=&Multiply_Float(
     A=&FloatFromCurve(
      FloatCurve=DataInterface() {propertyValues {Curve {keys [{interpMode=RCIM_Cubic, leaveTangent=8.1085, tangentMode=RCTM_Break, time=0.8568, value=1}, {tangentMode=RCTM_User, time=1, value=0}]}}}
      Scale Curve=0.25
     )
     B=&Divide_Float(
      A=&MakeFloatFromVector(Vector=$Particles.Scale, Channel=::NewEnumerator0)
      B=&RandomRangeFloat(Minimum=10, Maximum=15)
     )
    )
    Index 2 Param 2=&FloatFromCurve(
     FloatCurve=DataInterface() {propertyValues {Curve {keys [
         {interpMode=RCIM_Cubic, tangentMode=RCTM_Break, time=0.7, value=0}
         {arriveTangent=-0.0859, interpMode=RCIM_Cubic, tangentMode=RCTM_Break, time=1, value=-1}
        ]}}}
     Scale Curve=50
    )
    Index 2 Param 3=&Waveform(
     Global Input Position=&RandomRangeFloat(Minimum=0, Maximum=15)
     Global Phase=$Particles.Age
     "Waveform [1]"=::NewEnumerator12
     "[1] Phase Scale"=2
     "[1] Frequency"=1
     "[1] Amplitude Min/Max"=<0 1>
    )
    Index 2 Param 4=25
    Index 3 Param 1=&FloatFromCurve(
     FloatCurve=DataInterface() {propertyValues {Curve {keys [
         {interpMode=RCIM_Cubic, leaveTangent=-0.0025, tangentMode=RCTM_Break, time=0, value=0}
         {arriveTangent=1.8528, interpMode=RCIM_Cubic, tangentMode=RCTM_Break, time=1, value=1}
        ]}}}
     Scale Curve=&RandomRangeFloat(Minimum=15, Maximum=25)
    )
    Index 3 Param 2=&FloatFromCurve(FloatCurve=DataInterface() {propertyValues {Curve {keys [
         {interpMode=RCIM_Cubic, leaveTangent=-0.0025, tangentMode=RCTM_Break, time=0, value=0.5}
         {arriveTangent=-0.0316, interpMode=RCIM_Cubic, tangentMode=RCTM_Break, time=1, value=-0.2}
        ]}}})
    Index 3 Param 3=&FloatFromCurve(FloatCurve=DataInterface() {propertyValues {Curve {keys [
         {interpMode=RCIM_Cubic, leaveTangent=1.1034, tangentMode=RCTM_Break, time=0, value=-1}
         {arriveTangent=2.2399, tangentMode=RCTM_User, time=0.5, value=0.125}
        ]}}})
    Index 3 Param 4=&FloatFromCurve(FloatCurve=DataInterface() {propertyValues {Curve {keys [
         {interpMode=RCIM_Cubic, leaveTangent=3.0601, tangentMode=RCTM_Break, time=0.5, value=-0.5}
         {arriveTangent=3.2053, tangentMode=RCTM_User, time=1, value=1}
        ]}}})
    Module.Index0Param0Write=true
    Module.Index0Param1Write=true
    Module.Index0Param2Write=true
    Module.Index0Param3Write=true
    Module.Index1Param0Write=true
    Module.Index1Param1Write=true
    Module.Index1Param2Write=true
    Module.Index1Param3Write=true
    Module.Index2Param0Write=true
    Module.Index2Param1Write=true
    Module.Index2Param2Write=true
    Module.Index2Param3Write=true
    Module.Index3Param0Write=true
    Module.Index3Param1Write=true
    Module.Index3Param2Write=true
    Module.Index3Param3Write=true
    Write Parameter Index 1=true
    Write Parameter Index 2=true
    Write Parameter Index 3=true
   )
   &SolveForcesAndVelocity(Clamp Velocity=false, Limit Acceleration=false)
   &ScaleColor(
    Scale Alpha=&FloatFromCurve(FloatCurve=DataInterface() {propertyValues {Curve {keys [
         {tangentMode=RCTM_User, time=0, value=1}
         {arriveTangent=-0.0405, interpMode=RCIM_Cubic, leaveTangent=-0.0405, tangentMode=RCTM_User, time=0.7009, value=1}
         {arriveTangent=-8.0089, interpMode=RCIM_Cubic, leaveTangent=-8.0089, tangentMode=RCTM_User, time=1, value=0}
        ]}}})
    Scale RGB=&Lerp_Vector(
     A=&VectorFromFloat(Value=&FloatFromCurve(FloatCurve=DataInterface() {propertyValues {Curve {keys [<at 0.2771 1>,<at 0.3786 1>]}}}, Scale Curve=20))
     B=&VectorFromFloat(Value=&RandomRangeFloat(Minimum=0.3, Maximum=2, Evaluation Type=::NewEnumerator1))
     Alpha=&FloatFromCurve(FloatCurve=DataInterface() {propertyValues {Curve {keys [
          {interpMode=RCIM_Cubic, leaveTangent=2.5294, tangentMode=RCTM_Break, time=0, value=0}
          {arriveTangent=-0.2166, interpMode=RCIM_Cubic, tangentMode=RCTM_Break, time=0.4, value=1}
         ]}}})
    )
    ScaleA=true
    ScaleRGB=true
   )
   /SetVariables(Particles.MeshOrientation=&OrientMeshToVector(Look At Direction=$System.SysLightningFaceDir, Reference Axis Vector=<0 0 1>)) {isSetParameters=true}
   /SetVariables(Particles.MeshOrientation=&MultiplyQuaternion(
     Quaternion A=&MultiplyQuaternion(
      Quaternion A=&MultiplyQuaternion(
       Quaternion A=&MakeQuatFromAxisAngle(Axis=<0 0 1>, AngleInDegrees=-90)
       Quaternion B=&FindQuatBetweenTwoVectors(From Vector=<0 0 1>, To Vector=$Particles.Velocity)
      )
      Quaternion B=<0 0 0 1>
     )
     Quaternion B=&MakeQuatFromAxisAngle(Axis=<0 0 1>, AngleInDegrees=&RandomRangeFloat(Minimum=-360, Maximum=360))
    )) {enabled=false, isSetParameters=true, module=SetVariables_#28}
  ]
  rendererClasses [&NiagaraMeshRendererProperties]
  renderers [
   {rendererClass=&NiagaraMeshRendererProperties, rendererIndex=0}
   {rendererClass=&NiagaraMeshRendererProperties, rendererIndex=1}
  ]
 }
 SplitLightning001 {
  emitterUpdateScript [
   &EmitterState()
   &SpawnRate(
    SpawnRate=&FloatFromCurve(
     FloatCurve=DataInterface() {propertyValues {Curve {keys [<at 0 1 RCIM_Cubic>, {arriveTangent=21.1111, interpMode=RCIM_Cubic, leaveTangent=21.1111, time=0.45, value=2}, <at 0.9 20 RCIM_Cubic>]}}}
     CurveIndex=$System.Audio_DrumKick
     Scale Curve=0.5
    )
    Use Spawn Probability=false
   )
  ]
  particleSpawnScript [
   &InitializeParticle(
    Color=&RandomRangeLinearColor(Minimum=<rgba 1.2 1.2 1.2 0.5>, Maximum=<rgba 1.6 1.6 1.6 0.6>)
    Lifetime=&Lerp_Float(
     A=&RandomRangeFloat(Minimum=0.125, Maximum=0.3)
     B=&RandomRangeFloat(Minimum=0.25, Maximum=0.35)
     Alpha=$System.AudioIntensity
    )
    Mesh Scale Mode=::NewEnumerator3
    Mesh Uniform Scale=&Lerp_Float(A=0.2, B=0.4, Alpha=$System.AudioIntensity)
    UsePositionOffset=false
    Write Material Random=true
   )
   &AddVelocity(Velocity=<0 0 1>)
   &CameraOffset(Camera Offset Amount=5) {enabled=false}
  ]
  particleUpdateScript [
   &ParticleState()
   /SetVariables(Particles.Position=&AddVectorToPosition(
     Position=$System.SkullPosition
     Vector=&TransformPosition(Position=$Particles.Initial.Position, Source Space=::World, Destination Space=::Simulation)
    )) {enabled=false, isSetParameters=true}
   &ShapeLocation(
    Additional Quaternion Rotation=true
    Additional Yaw / Pitch / Roll=<0 0.25 0>
    Cylinder Distribution=::NewEnumerator2
    Cylinder Height=10
    Cylinder Radius=16
    MaskSpawnGroup=false
    Offset Mode=::NewEnumerator0
    Rotation Mode=::NewEnumerator3
    Rotation Quaternion=$System.SkullRotation
    Shape Origin=$System.SkullPosition
    Shape Primitive=::NewEnumerator1
    U Position=&RandomRangeFloat(Minimum=0.4, Maximum=1.1)
    UseOwnerScale=true
    V Position=&RandomRangeFloat(Minimum=0, Maximum=1)
   )
   &ScaleMeshSize(Scale Factor=&RandomRangeVector(Minimum=<1.2 1.2 3>, Maximum=<1.8 1.8 1>)) {module=ScaleMeshSize001}
   &ScaleMeshSize(Scale Factor=&MakeVector(
     X=1
     Y=1
     Z=&FloatFromCurve(
      FloatCurve=DataInterface() {propertyValues {Curve {keys [
          {interpMode=RCIM_Cubic, leaveTangent=0.0966, tangentMode=RCTM_Break, time=0, value=1}
          {arriveTangent=0.0951, interpMode=RCIM_Cubic, leaveTangent=-0.1745, tangentMode=RCTM_Break, time=1, value=1.07}
         ]}}}
      Scale Curve=&Lerp_Float(A=0.45, B=1.25, Alpha=&Max_Float(A:NiagaraFloat=$System.AudioIntensity, B:NiagaraFloat=$System.Dive))
     )
    ))
   &DynamicMaterialParameters(
    Index 0 Param 4=&MakeCustomFloatFromBool(
     Bool=false
     True Float=10
     False Float=&SelectFloatFromArray(
      Array Sampling Mode=::NewEnumerator0
      Direct Array Index=?
      Float Selection Array=DataInterface() {propertyValues {FloatData [3,9,9]}}
      Random Seed=0
     )
    )
    Index 1 Param 1=1
    Index 1 Param 2=2
    Index 1 Param 3=&FloatFromCurve(FloatCurve=DataInterface() {propertyValues {Curve {keys [{interpMode=RCIM_Cubic, tangentMode=RCTM_Break, time=0.5, value=-0.5}, <at 1 0.1>]}}})
    Index 2 Param 1=&Multiply_Float(
     A=&FloatFromCurve(
      FloatCurve=DataInterface() {propertyValues {Curve {keys [{interpMode=RCIM_Cubic, leaveTangent=8.1085, tangentMode=RCTM_Break, time=0.8568, value=1}, {tangentMode=RCTM_User, time=1, value=0}]}}}
      Scale Curve=0.5
     )
     B=&Divide_Float(
      A=&MakeFloatFromVector(Vector=$Particles.Scale, Channel=::NewEnumerator0)
      B=&RandomRangeFloat(Minimum=10, Maximum=15)
     )
    )
    Index 2 Param 2=&FloatFromCurve(
     FloatCurve=DataInterface() {propertyValues {Curve {keys [
         {interpMode=RCIM_Cubic, tangentMode=RCTM_Break, time=0.7, value=0}
         {arriveTangent=-0.0859, interpMode=RCIM_Cubic, tangentMode=RCTM_Break, time=1, value=-1}
        ]}}}
     Scale Curve=50
    )
    Index 2 Param 3=&Waveform(
     Global Input Position=&RandomRangeFloat(Minimum=0, Maximum=15)
     Global Phase=$Particles.Age
     "Waveform [1]"=::NewEnumerator12
     "[1] Phase Scale"=2
     "[1] Frequency"=1
     "[1] Amplitude Min/Max"=<0 1>
    )
    Index 2 Param 4=25
    Index 3 Param 1=&FloatFromCurve(
     FloatCurve=DataInterface() {propertyValues {Curve {keys [
         {interpMode=RCIM_Cubic, leaveTangent=-0.0025, tangentMode=RCTM_Break, time=0, value=0}
         {arriveTangent=1.8528, interpMode=RCIM_Cubic, tangentMode=RCTM_Break, time=1, value=1}
        ]}}}
     Scale Curve=&RandomRangeFloat(Minimum=15, Maximum=25)
    )
    Index 3 Param 2=&FloatFromCurve(FloatCurve=DataInterface() {propertyValues {Curve {keys [
         {interpMode=RCIM_Cubic, leaveTangent=-0.0025, tangentMode=RCTM_Break, time=0, value=0.5}
         {arriveTangent=-0.0316, interpMode=RCIM_Cubic, tangentMode=RCTM_Break, time=1, value=-0.2}
        ]}}})
    Index 3 Param 3=&FloatFromCurve(FloatCurve=DataInterface() {propertyValues {Curve {keys [
         {interpMode=RCIM_Cubic, leaveTangent=1.1034, tangentMode=RCTM_Break, time=0, value=-1}
         {arriveTangent=2.2399, tangentMode=RCTM_User, time=0.5, value=0.125}
        ]}}})
    Index 3 Param 4=&FloatFromCurve(FloatCurve=DataInterface() {propertyValues {Curve {keys [
         {interpMode=RCIM_Cubic, leaveTangent=3.0601, tangentMode=RCTM_Break, time=0.5, value=-0.5}
         {arriveTangent=3.2053, tangentMode=RCTM_User, time=1, value=1}
        ]}}})
    Module.Index0Param0Write=true
    Module.Index0Param1Write=true
    Module.Index0Param2Write=true
    Module.Index0Param3Write=true
    Module.Index1Param0Write=true
    Module.Index1Param1Write=true
    Module.Index1Param2Write=true
    Module.Index1Param3Write=true
    Module.Index2Param0Write=true
    Module.Index2Param1Write=true
    Module.Index2Param2Write=true
    Module.Index2Param3Write=true
    Module.Index3Param0Write=true
    Module.Index3Param1Write=true
    Module.Index3Param2Write=true
    Module.Index3Param3Write=true
    Write Parameter Index 1=true
    Write Parameter Index 2=true
    Write Parameter Index 3=true
   )
   &SolveForcesAndVelocity(Clamp Velocity=false, Limit Acceleration=false)
   &ScaleColor(
    Scale Alpha=&FloatFromCurve(FloatCurve=DataInterface() {propertyValues {Curve {keys [
         {tangentMode=RCTM_User, time=0, value=1}
         {arriveTangent=-0.0405, interpMode=RCIM_Cubic, leaveTangent=-0.0405, tangentMode=RCTM_User, time=0.7009, value=1}
         {arriveTangent=-8.0089, interpMode=RCIM_Cubic, leaveTangent=-8.0089, tangentMode=RCTM_User, time=1, value=0}
        ]}}})
    Scale RGB=&Multiply_VectorByFloat(
     Vector=&Lerp_Vector(
      A=&VectorFromFloat(Value=&FloatFromCurve(FloatCurve=DataInterface() {propertyValues {Curve {keys [<at 0.2771 1>,<at 0.3786 1>]}}}, Scale Curve=20))
      B=&VectorFromFloat(Value=&RandomRangeFloat(Minimum=0.3, Maximum=2, Evaluation Type=::NewEnumerator1))
      Alpha=&FloatFromCurve(FloatCurve=DataInterface() {propertyValues {Curve {keys [
           {interpMode=RCIM_Cubic, leaveTangent=2.5294, tangentMode=RCTM_Break, time=0, value=0}
           {arriveTangent=-0.2166, interpMode=RCIM_Cubic, tangentMode=RCTM_Break, time=0.4, value=1}
          ]}}})
     )
     Float=&Lerp_Float(A=0.5, B=1.25, Alpha=$System.AudioBrightnessMultiplier)
    )
    ScaleA=true
    ScaleRGB=true
   )
   /SetVariables(Particles.FacingDir:Vector3f=&FN_FrontEndVector3Modifier(
     Gameplay=&NormalizeVector(
      Vector To Normalize=&Add_Vector(
       A=&SubtractVectorFromPosition(
        Position=$Particles.Position
        Vector=&SubtractVectorFromPosition(Position=$System.SkullPosition, Vector=<0 0 100>)
       )
       B=<0 0 1>
      )
      Safe Normalize=true
      Safe Fallback Threshold=0
      Safe Fallback Vector=<0 0 0.001>
     )
     Front End=&NormalizeVector(
      Vector To Normalize=&Add_Vector(A=&RotateVectorByQuaternion(Vector To Rotate=<-1 0 0>, Quaternion=$System.SkullRotation), B=<0 0 1>)
      Safe Normalize=true
      Safe Fallback Threshold=0
      Safe Fallback Vector=<0 0 0.001>
     )
    )) {isSetParameters=true}
   /SetVariables(Particles.FacingDir:Vector3f=<0 0 1>) {enabled=false, isSetParameters=true}
   /SetVariables(Particles.MeshOrientation=&MultiplyQuaternion(
     Quaternion A=&OrientMeshToVector(Look At Direction=$Particles.FacingDir, Reference Axis Vector=<0 0 1>)
     Quaternion B=&MakeQuatFromAxisAngle(Axis=<0 0 1>, AngleInDegrees=&RandomRangeFloat(Minimum=0, Maximum=360))
    )) {isSetParameters=true, module=SetVariables_#28}
   /SetVariables(Particles.MeshOrientation=&OrientMeshToVector(Look At Direction=$Particles.FacingDir, Reference Axis Vector=<0 0 1>)) {enabled=false, isSetParameters=true}
  ]
  rendererClasses [&NiagaraMeshRendererProperties]
  renderers [
   {rendererClass=&NiagaraMeshRendererProperties, rendererIndex=0}
   {rendererClass=&NiagaraMeshRendererProperties, rendererIndex=1}
  ]
 }
}
summary {
 systemName=NS_Contrail_GreatPool
 emitters [
  {
   emitterName=MainSkull
   rendererClasses [
    &NiagaraMeshRendererProperties
    &NiagaraSpriteRendererProperties
    &NiagaraLightRendererProperties
   ]
  }
  {bEnabled=false, emitterName=FN_Debug, rendererClasses [&NiagaraSpriteRendererProperties]}
  {bEnabled=false, emitterName=EnergySplineBolts, rendererClasses [&NiagaraMeshRendererProperties]}
  {emitterName=SplitLightning001, rendererClasses [&NiagaraMeshRendererProperties]}
  {emitterName=Lightning_Arcs, rendererClasses [&NiagaraSpriteRendererProperties]}
  {emitterName=MainElectricity, rendererClasses [&NiagaraMeshRendererProperties]}
  {emitterName=PointHelpers, rendererClasses [&NiagaraSpriteRendererProperties]}
  {emitterName=Bolt_Meshes, rendererClasses [&NiagaraMeshRendererProperties]}
  {emitterName=Sparks, simTarget=GPUComputeSim, rendererClasses [&NiagaraSpriteRendererProperties]}
  {bEnabled=false, emitterName=SplitLightning, rendererClasses [&NiagaraMeshRendererProperties]}
  {bEnabled=false, emitterName=Empty, rendererClasses [&NiagaraSpriteRendererProperties]}
  {emitterName=ElectricCoils, rendererClasses [&NiagaraMeshRendererProperties]}
 ]
 userVariables [
  User.bIsFrontEnd {type=NiagaraBool}
  User.bIsFrontEndPreview {type=NiagaraBool}
  User.bIsOffhand {type=NiagaraBool}
  User.FXBoundScalar {type=NiagaraFloat, defaultValue=1, description=""}
  User.ParaGlideLeanAlpha {type=NiagaraFloat, defaultValue=0, description=""}
  User.RibbonIDOffset {type=NiagaraInt32}
  User.TargetIsLaserGrid {type=NiagaraBool}
 ]
}
